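# Translucent, lit Niagara sprite material for small snow particles in the Weeping Woods Festive character FX.
# Time-of-day dark colors blend toward white by the dot product of vertex normal and sun direction, then multiply time-of-day main colors.
# That tint feeds a shared emissive adjustment function, alongside particle color.
# A radial sphere mask on the UVs, multiplied by particle alpha, forms a round dot without textures.
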Material /BRCosmetics/Characters/Player/Male/Medium/Bodies/M_MED_WeepingWoodsFestive/FX/M_WeepingWoodsFestive_SnowSpriteSmall
graph {
 properties {bUsedWithNiagaraSprites=true, blendMode=BLEND_Translucent, pixelDepthOffsetMode=PDOM_Legacy, usedShadingModels=Default Lit}
 nodes [
  n1 = MaterialFunctionCall(
   Color_Morning=VectorParameter() {parameterName=DarkColor_Morning, defaultValue=<rgba 0.56 0.63 0.7 0>}.RGB
   Color_Day=VectorParameter() {parameterName=DarkColor_Day, defaultValue=<rgba 0.4588 0.5098 0.6118 1>}.RGB
   Color_Evening=VectorParameter() {parameterName=DarkColor_Evening, defaultValue=<rgba 0.64 0.6016 0.6208 0>}.RGB
   Color_Night=VectorParameter() {parameterName=DarkColor_Night, defaultValue=<rgba 0.56 0.6067 0.7 1>}.RGB
  ) {materialFunction=&Color_Time_of_Day}
  n2 = Constant() {r=1}
  n3 = DotProduct(
   A=VertexNormalWS()
   B=ComponentMask(None=MaterialFunctionCall() {materialFunction=/Game/Global/MaterialFunctions/FNCollectionParameters/MF_SunParameters_FNC}.Direction) {b=true, g=true, r=true}
  )
  n4 = VectorParameter() {parameterName=MainColor_Morning, defaultValue=<rgba 0.85 0.5466 0.544 0>}
  n5 = VectorParameter() {parameterName=MainColor_Day, defaultValue=<rgba 1 1 1 0>}
  n6 = VectorParameter() {parameterName=MainColor_Evening, defaultValue=<rgba 0.76 0.5067 0.456 1>}
  n7 = VectorParameter() {parameterName=MainColor_Night, defaultValue=<rgba 0.36 0.54 0.9 1>}
  MaterialFunctionCall(Emissive=Multiply(
    A=LinearInterpolate(A=@n1.Result, B=@n2, Alpha=@n3)
    B=MaterialFunctionCall(Color_Morning=@n4.RGB, Color_Day=@n5.RGB, Color_Evening=@n6.RGB, Color_Night=@n7.RGB) {materialFunction=&Color_Time_of_Day}.Result
   )) {materialFunction=&MF_FX_EmissiveAdjustments}
  n8 = ParticleColor()
  Multiply(A=SphereMask(A=TextureCoordinate(), B=Constant() {r=0.5}) {attenuationRadius=0.4}, B=@n8.A)
  Multiply(B=@n8.RGB)
 ]
}
