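# Material function for the Weeping Woods Festive body cosmetic that adds blinking festive lights to incoming material attributes.
# A mask input, previewed with the body FX texture alpha, is thresholded into a lights mask and also kept as an offset mask.
# Offset time drives a clamped sine blink, with parameters for speed, switch, enable and on/off colors, which sets base color and emissive.
# Time-of-day and thin-film spectrum branches are present but not connected to the output.
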MaterialFunction /BRCosmetics/Characters/Player/Male/Medium/Bodies/M_MED_WeepingWoodsFestive/Materials/MF_Festive_BlinkingLights
graph {nodes [
  n1 = GetMaterialAttributes(None=FunctionInput() {inputType=FunctionInput_MaterialAttributes}) {attributeGetTypes [#1, #2]}
  n2 = NamedRerouteUsage()
  n3 = Multiply(A=NamedRerouteUsage(), B=VectorParameter() {group=Lights, parameterName=LightColor_Off, defaultValue=<rgba 1 1 1 1>}.RGB)
  n4 = NamedRerouteUsage()
  n5 = Time()
  n6 = NamedRerouteUsage()
  n7 = Add(A=Multiply(A=Time() {bOverride_Period=true, period=20}, B=@n6), B=@n6)
  n8 = Sine(None=Multiply(A=Add(A=@n5, B=@n7), B=ScalarParameter() {defaultValue=1, group=Lights, parameterName=BlinkSpeed}))
  n9 = LinearInterpolate(
   A=Constant() {r=1}
   B=Multiply(A=Saturate(None=@n8), B=ScalarParameter() {defaultValue=1, group=Lights, parameterName=LightSwitch, sliderMax=1})
   Alpha=ScalarParameter() {defaultValue=1, group=Lights, parameterName=EnableLightBlink, sliderMax=1}
  )
  n10 = Multiply(
   A=NamedRerouteUsage()
   B=Multiply(A=@n9, B=VectorParameter() {group=Lights, parameterName=LightColor_On, defaultValue=<rgba 50 9.1667 0 1>}.RGB)
  )
  FunctionOutput(None=SetMaterialAttributes!(MaterialAttributes=@n1.MaterialAttributes, BaseColor=LinearInterpolate(A=@n2, B=@n3, Alpha=@n4), EmissiveColor=@n10) {attributeSetTypes [#1, #2]})
  n11 = Multiply(A=@n10) {constB=0.5}
  MaterialFunctionCall(Color_Morning=@n10, Color_Day=@n10, Color_Evening=@n11, Color_Night=@n11) {materialFunction=&Color_Time_of_Day}
  n12 = NamedRerouteUsage()
  Multiply(A=TextureSample(UVs=Add(A=@n12) {constB=0}) {texture=/Game/Global/Textures/Default/Thinfilm/T_ThinFilm_Spectrum_COLOR}.RGB, B=@n9)
  n13 = TextureSample() {
   samplerType=SAMPLERTYPE_Masks
   texture=/BRCosmetics/Characters/Player/Male/Medium/Bodies/M_MED_WeepingWoodsFestive/Textures/T_WeepingWoodsFestive_Body_FX
  }
  n14 = FunctionInput(Preview=@n13.A) {inputName=Mask, inputType=FunctionInput_Scalar, sortPriority=10}
  NamedRerouteDeclaration(None=Step(X=NamedRerouteDeclaration(None=@n14) {name=LightOffsetMask, nodeColor=<rgba 1 0.2578 0 1>}) {constX=0.1, constY=0.17}) {name=LightsMask, nodeColor=<rgba 0.7812 0 1 1>}
  NamedRerouteDeclaration(None=@n1.EmissiveColor) {name=InEmissive, nodeColor=<rgba 0.1953 0 1 1>}
  NamedRerouteDeclaration(None=@n1.BaseColor) {name=InBaseColor, nodeColor=<rgba 0.1484 1 0 1>}
 ]}
