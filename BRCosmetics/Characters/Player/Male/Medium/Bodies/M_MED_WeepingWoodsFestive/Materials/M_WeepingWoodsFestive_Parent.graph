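# Parent character material for the WeepingWoodsFestive skin, wrapping optional anisotropy and blinking festive lights between the shared 2023 character PreFX and PostFX functions.
# A switch picks a custom FX texture or the Tech Art Mask to drive both effects.
# Anisotropic shading sets roughness, metalness and a tiled sphere normal on UV1.
# Its tangent switches between front and back vectors by the sign of a local-space axis, likely pre-skinned position.
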Material /BRCosmetics/Characters/Player/Male/Medium/Bodies/M_MED_WeepingWoodsFestive/Materials/M_WeepingWoodsFestive_Parent
graph {
 properties {
  bUseMaterialAttributes=true
  bUsedWithMeshDeformer=true
  bUsedWithMorphTargets=true
  bUsedWithSkeletalMesh=true
  pixelDepthOffsetMode=PDOM_Legacy
  previewMesh=/BRCosmetics/Characters/Player/Male/Medium/Bodies/M_MED_WeepingWoodsFestive/Meshes/M_MED_WeepingWoodsFestive
  shadingModel=MSM_FromMaterialExpression
  usedShadingModels=Default Lit | Subsurface | Clear Coat | Subsurface Profile | Cloth
 }
 nodes [
  n1 = TextureSampleParameter2D() {
   group=WeepingWoodsFestive
   parameterName=CustomFXMask
   samplerType=SAMPLERTYPE_Masks
   texture=/BRCosmetics/Characters/Player/Male/Medium/Bodies/M_MED_WeepingWoodsFestive/Textures/T_WeepingWoodsFestive_Body_FX
  }
  n2 = MaterialFunctionCall() {materialFunction=&MF_CharacterParent_2023_PreFX}
  n3 = StaticSwitchParameter(True=@n1.RGBA, False=@n2.Tech Art Mask) {group=WeepingWoodsFestive, parameterName=UseCustomMask}
  n4 = StaticSwitchParameter(
   True=BlendMaterialAttributes(A=NamedRerouteUsage(), B=NamedRerouteUsage(), Alpha=ChannelMaskParameter(None=@n3) {group=06_ Anisotropic, parameterName=AnisoMask}.R)
   False=NamedRerouteUsage()
  ) {parameterName=EnableAnisotropy}
  n5 = ChannelMaskParameter(None=@n3) {group=WeepingWoodsFestive, parameterName=LightMask}
  MaterialFunctionCall(
   Material=StaticSwitchParameter(
    True=MaterialFunctionCall!(In=@n4, Mask=@n5.R) {materialFunction=&/MF_Festive_BlinkingLights}.Result
    False=@n4
   ) {group=WeepingWoodsFestive, parameterName=Use Lights}
   M=@n2.M
   Pre Skinned Local Position=@n2.M
   Distance Blend=@n2.M
   Fresnel=@n2.M
  ) {materialFunction=&MF_CharacterParent_2023_PostFX}
  n6 = TextureSample(UVs=Multiply(A=TextureCoordinate() {coordinateIndex=1}, B=ScalarParameter() {defaultValue=1, group=06_ Anisotropic, parameterName=AnisoNormalTiling}) {constB=5}) {
   samplerType=SAMPLERTYPE_Normal
   texture=&/T_AnisoSpheres_N
  }
  n7 = ScalarParameter() {defaultValue=1, group=06_ Anisotropic, parameterName=AnisoNormalIntensity}
  n8 = MaterialFunctionCall(X=Multiply(A=@n6.R, B=@n7), Y=Multiply(A=@n6.G, B=@n7), Z=@n6.B) {materialFunction=&MakeFloat3}
  MaterialFunctionCall(BaseNormal=NamedRerouteUsage(), AdditionalNormal=@n8.Result) {materialFunction=&BlendAngleCorrectedNormals}
  n9 = GetMaterialAttributes(None=@n2.Result) {attributeGetTypes [#1]}
  NamedRerouteDeclaration(None=@n9.Normal) {name=InNormal, nodeColor=<rgba 0.8057 0.2699 1 1>}
  NamedRerouteDeclaration(None=@n9.MaterialAttributes) {name=InMaterialAttributes, nodeColor=<rgba 0 1 0.4844 1>}
  n10 = NamedRerouteUsage()
  n11 = Ceil(None=ComponentMask(None=Normalize(VectorInput=@n10)) {g=true})
  n12 = LinearInterpolate(
   A=VectorParameter() {group=06_ Anisotropic, parameterName=AnisotropicTangentWeight_Front, defaultValue=<rgba -1 -1 -1 1>}.RGB
   B=VectorParameter() {group=06_ Anisotropic, parameterName=AnisotropicTangentWeight_Back, defaultValue=<rgba -1 -1 -1 1>}.RGB
   Alpha=OneMinus(None=Saturate(None=@n11))
  )
  n13 = SetMaterialAttributes!(
   MaterialAttributes=NamedRerouteUsage()
   Tangent=Normalize(VectorInput=Transform(None=@n12) {transformSourceType=TRANSFORMSOURCE_Local, transformType=TRANSFORM_Tangent})
   Anisotropy=ScalarParameter() {defaultValue=-1, group=06_ Anisotropic, parameterName=AnisotropyWeight, sliderMax=1, sliderMin=-1}
   Roughness=ScalarParameter() {group=06_ Anisotropic, parameterName=AnisoRoughness}
   Normal=@n8.Result
   Metallic=ScalarParameter() {defaultValue=0.5, group=06_ Anisotropic, parameterName=AnisoMetalness}
  ) {attributeSetTypes [
    #2
    #3
    #4
    #1
    #5
   ]}
  NamedRerouteDeclaration(None=@n13) {name=AnisoShading, nodeColor=<rgba 0.7812 1 0 1>}
  NamedRerouteDeclaration(None=@n2.Pre skinned Local Position) {name=PreSkinnedLocalPosition, nodeColor=<rgba 0 0.5078 1 1>}
 ]
}
