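# Control Rig for the WeepingWoodsFestive male body cosmetic that procedurally poses its secondary bones.
# A sequence first parent-constrains ornament and sash ornament bones to weighted blends of spine and pine bones.
# Spherical pose readers on each thigh then weight Modify Transforms offsets on skirt bones, plus a candy cane for the left leg.
# Upper-arm pose readers for shoulder spike correctives exist but are not wired into the execution sequence.
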ControlRigBlueprint /BRCosmetics/Characters/Player/Male/Medium/Bodies/M_MED_WeepingWoodsFestive/Meshes/M_MED_WeepingWoodsFestive_CtrlRig
graphs [
 RigVMFunctionLibrary {links [], nodes []}
 RigVMFunction_Sequence_1_ContainedGraph {nodes [
   n1 = RigVMFunctionEntryNode()
   RigVMFunction_Sequence = &RigVMFunction_Sequence(ExecuteContext=@n1.ExecuteContext)
   n2 = &RigVMFunction_Sequence(ExecuteContext=@RigVMFunction_Sequence.B)
   n3 = &RigVMFunction_Sequence(ExecuteContext=@n2.B)
   n4 = &RigVMFunction_Sequence(ExecuteContext=@n3.B)
   n5 = &RigVMFunction_Sequence(ExecuteContext=@n4.B)
   RigVMFunctionReturnNode(A=@RigVMFunction_Sequence.A, B=@n2.A, C=@n3.A, D=@n4.A, E=@n5.A, F=@n5.B)
  ]}
 RigVMModel {nodes [
   RigUnit_BeginExecution = &RigUnit_BeginExecution()
   RigVMFunction_Sequence = RigVMAggregateNode(ExecuteContext=@RigUnit_BeginExecution.ExecutePin)
   &RigUnit_ParentConstraint(
    AdvancedSettings=(RotationOrderForFilter=XZY)
    Child=(Type=Bone,Name="CR_OrnamentA1")
    Parents="((Item=(Type=Bone,Name=\"AD_Sm_pine_rowA_Q1\"),Weight=0.0485),(Item=(Type=Bone,Name=\"spine_05\"),Weight=0.0535),(Item=(Type=Bone,Name=\"AD_Sm_pine_rowA_O1\"),Weight=0.0599),(Item=(Type=Bone,Name=\"AD_Lg_pine_rowB_C1\"),Weight=0.0747),(Item=(Type=Bone,Name=\"spine_ ...(truncated)"
    Weight=1
    ExecutePin=@RigVMFunction_Sequence.A
   )
   > &RigUnit_ParentConstraint(
    AdvancedSettings=(RotationOrderForFilter=XZY)
    Child=(Type=Bone,Name="CR_SashOrnamentA1")
    Parents=((Item=(Type=Bone,Name="spine_01"),Weight=0.0035),(Item=(Type=Bone,Name="spine_03"),Weight=0.1101),(Item=(Type=Bone,Name="spine_05"),Weight=0.255),(Item=(Type=Bone,Name="spine_04"),Weight=0.6314))
    Weight=1
   )
   > &RigUnit_ParentConstraint(
    AdvancedSettings=(RotationOrderForFilter=XZY)
    Child=(Type=Bone,Name="CR_SashOrnamentB1")
    Parents=((Item=(Type=Bone,Name="spine_01"),Weight=0.0453),(Item=(Type=Bone,Name="spine_05"),Weight=0.0592),(Item=(Type=Bone,Name="spine_03"),Weight=0.4141),(Item=(Type=Bone,Name="spine_04"),Weight=0.4814))
    Weight=1
   )
   n1 = &RigUnit_SphericalPoseReader(
    ActiveRegionScaleFactors=(PositiveWidth=3,NegativeWidth=2,PositiveHeight=1,NegativeHeight=3)
    ActiveRegionSize=0.1
    Debug=(bDrawDebug=false,bDraw2D=false,bDrawLocalAxes=false,DebugScale=25,DebugSegments=20,DebugThickness=0.25)
    DriverAxis=(X=-1,Y=0,Z=0)
    DriverItem=(Type=Bone,Name="thigh_l")
    FalloffRegionScaleFactors=(PositiveWidth=1,NegativeWidth=1,PositiveHeight=1,NegativeHeight=1)
    FalloffSize=0.5
    FlipHeightScaling=False
    FlipWidthScaling=False
    OptionalParentItem=(Type=Bone,Name="None")
    RotationOffset=(X=120,Y=0,Z=0)
    ExecutePin=@RigVMFunction_Sequence.B
   )
   > &RigUnit_ModifyTransforms(
    ItemToModify="((Item=(Type=Bone,Name=\"MSTR_skirt_fr\"),Transform=(Rotation=(X=-0.0399,Y=0.0099,Z=-0.5592,W=0.828),Translation=(X=3.2439,Y=0.1826,Z=-0.2087),Scale3D=(X=1,Y=1,Z=1))),(Item=(Type=Bone,Name=\"MSTR_skirtA_l\"),Transform=(Rotation=(X=-0.1835,Y=-0.2107,Z=-0.5903,W ...(truncated)"
    WeightMinimum=0
    Weight=@n1.OutputParam
   )
   > &RigUnit_ModifyTransforms(
    ItemToModify=((Item=(Type=Bone,Name="CR_CandyCane1"),Transform=(Rotation=(X=0.6654,Y=0,Z=0,W=0.7465),Translation=(X=0,Y=0,Z=0),Scale3D=(X=1,Y=1,Z=1))))
    WeightMinimum=0
    Weight=@n1.OutputParam
   )
   n2 = &RigUnit_SphericalPoseReader(
    ActiveRegionScaleFactors=(PositiveWidth=2.6,NegativeWidth=3,PositiveHeight=2,NegativeHeight=3.3)
    ActiveRegionSize=0.1
    Debug=(bDrawDebug=false,bDraw2D=false,bDrawLocalAxes=false,DebugScale=25,DebugSegments=20,DebugThickness=0.25)
    DriverAxis=(X=1,Y=0,Z=0)
    DriverItem=(Type=Bone,Name="thigh_r")
    FalloffRegionScaleFactors=(PositiveWidth=1,NegativeWidth=1,PositiveHeight=1,NegativeHeight=1)
    FalloffSize=0.6
    FlipHeightScaling=False
    FlipWidthScaling=False
    OptionalParentItem=(Type=Bone,Name="None")
    RotationOffset=(X=-60,Y=0,Z=-30)
    ExecutePin=@RigVMFunction_Sequence.C
   )
   > &RigUnit_ModifyTransforms(
    ItemToModify="((Item=(Type=Bone,Name=\"MSTR_skirt_fr\"),Transform=(Rotation=(X=-0.1515,Y=-0.066,Z=-0.0628,W=0.9843),Translation=(X=-9.7582,Y=-9.0858,Z=3.3301),Scale3D=(X=1,Y=1,Z=1))),(Item=(Type=Bone,Name=\"MSTR_skirtA_r\"),Transform=(Rotation=(X=-0.0965,Y=0.2675,Z=-0.4831, ...(truncated)"
    WeightMinimum=0
    Weight=@n2.OutputParam
   )
   n3 = &RigUnit_SphericalPoseReader(
    ActiveRegionScaleFactors=(PositiveWidth=0.5,NegativeWidth=3,PositiveHeight=1.5,NegativeHeight=1.5)
    ActiveRegionSize=0.1
    Debug=(bDrawDebug=true,DebugScale=25,DebugSegments=20,DebugThickness=0.25)
    DriverAxis=(X=-1,Y=0,Z=0)
    DriverItem=(Type=Bone,Name="upperarm_r")
    ExecutePin=()
    FalloffRegionScaleFactors=(PositiveWidth=1,NegativeWidth=1,PositiveHeight=1,NegativeHeight=1)
    FalloffSize=0.5
    FlipWidthScaling=true
    OptionalParentItem=(Type=Bone)
    RotationOffset=(X=90,Y=180,Z=180)
   )
   > &RigUnit_ModifyTransforms(
    ItemToModify="((Item=(Type=Bone,Name=\"CR_shoulder_r_A\"),Transform=(Rotation=(X=0.1017,Y=0,Z=0,W=0.9948),Translation=(X=0,Y=0,Z=0),Scale3D=(X=1,Y=1,Z=1))),(Item=(Type=Bone,Name=\"CR_shoulder_r_B\"),Transform=(Rotation=(X=0,Y=0,Z=0.2995,W=0.9541),Translation=(X=0,Y=0,Z=0),S ...(truncated)"
    WeightMinimum=0
    Weight=@n3.OutputParam
   )
   n4 = &RigUnit_SphericalPoseReader(
    ActiveRegionScaleFactors=(PositiveWidth=2.5,NegativeWidth=3,PositiveHeight=3,NegativeHeight=3)
    ActiveRegionSize=0.1
    Debug=(bDrawDebug=false,DebugScale=25,DebugSegments=20,DebugThickness=0.25)
    DriverAxis=(X=0,Y=0,Z=-1)
    DriverItem=(Type=Bone,Name="upperarm_r")
    ExecutePin=()
    FalloffRegionScaleFactors=(PositiveWidth=1,NegativeWidth=1,PositiveHeight=1,NegativeHeight=1)
    FalloffSize=0.5
    FlipHeightScaling=true
    OptionalParentItem=(Type=Bone)
    RotationOffset=(X=180,Y=0,Z=180)
   )
   > &RigUnit_ModifyTransforms(
    ItemToModify="((Item=(Type=Bone,Name=\"CR_shoulder_r_A\"),Transform=(Rotation=(X=0,Y=0,Z=0,W=1),Translation=(X=0,Y=0,Z=0),Scale3D=(X=1,Y=1,Z=1))),(Item=(Type=Bone,Name=\"CR_shoulder_r_B\"),Transform=(Rotation=(X=0,Y=0,Z=0,W=1),Translation=(X=0,Y=0,Z=0),Scale3D=(X=1,Y=1,Z=1) ...(truncated)"
    WeightMinimum=0
    Weight=@n4.OutputParam
   )
   n5 = &RigUnit_SphericalPoseReader(
    ActiveRegionScaleFactors=(PositiveWidth=2.5,NegativeWidth=3,PositiveHeight=3,NegativeHeight=3)
    ActiveRegionSize=0.1
    Debug=(bDrawDebug=false,DebugScale=25,DebugSegments=20,DebugThickness=0.25)
    DriverAxis=(X=1,Y=0,Z=0)
    DriverItem=(Type=Bone,Name="upperarm_l")
    ExecutePin=()
    FalloffRegionScaleFactors=(PositiveWidth=1,NegativeWidth=1,PositiveHeight=1,NegativeHeight=1)
    FalloffSize=0.5
    OptionalParentItem=(Type=Bone)
    RotationOffset=(X=0,Y=0,Z=0)
   )
   > &RigUnit_ModifyTransforms(
    ItemToModify="((Item=(Type=Bone,Name=\"None\"),Transform=(Rotation=(X=0,Y=0,Z=0,W=1),Translation=(X=0,Y=0,Z=0),Scale3D=(X=1,Y=1,Z=1))),(Item=(Type=Bone,Name=\"CR_shoulder_l_A\"),Transform=(Rotation=(X=0,Y=0,Z=0,W=1),Translation=(X=0,Y=0,Z=0),Scale3D=(X=1,Y=1,Z=1))),(Item=(T ...(truncated)"
    WeightMinimum=0
    Weight=@n5.OutputParam
   )
   &RigUnit_SphericalPoseReader(
    ActiveRegionScaleFactors=(PositiveWidth=0.5,NegativeWidth=3,PositiveHeight=1.5,NegativeHeight=1.5)
    ActiveRegionSize=0.1
    Debug=(bDrawDebug=false,DebugScale=25,DebugSegments=20,DebugThickness=0.25)
    DriverAxis=(X=1,Y=0,Z=0)
    DriverItem=(Type=Bone,Name="upperarm_l")
    ExecutePin=()
    FalloffRegionScaleFactors=(PositiveWidth=1,NegativeWidth=1,PositiveHeight=1,NegativeHeight=1)
    FalloffSize=0.4
    OptionalParentItem=(Type=Bone)
    RotationOffset=(X=90,Y=0,Z=0)
   )
  ]}
]
hierarchy [
 root {loc=<0 0 0>, rot=<0 0 0 1>}
 pelvis {parent=root, loc=<0 -0.7071 93.1191>, rot=<0.0127 0.707 -0.0127 -0.707>}
 spine_01 {parent=pelvis, loc=<6.7494 -0.6388 0>, rot=<0 0 0.034 -0.9994>}
 spine_02 {parent=spine_01, loc=<7.2518 -0.035 0>, rot=<0 0 0.0238 -0.9997>}
 spine_03 {parent=spine_02, loc=<6.0104 0 0>, rot=<0 0 0.0149 -0.9999>}
 spine_04 {parent=spine_03, loc=<8.1024 0 0>, rot=<0 0 -0.0132 -0.9999>}
 spine_05 {parent=spine_04, loc=<10.0146 0 0>, rot=<0 0 -0.0015 -1>}
 attach_cape {parent=spine_05, loc=<9.017 -11.714 0>, rot=<0.0616 -0.7044 -0.0616 -0.7044>}
 neck_01 {parent=spine_05, loc=<11.3138 0.1529 0>, rot=<0 0 -0.2003 -0.9797>}
 neck_02 {parent=neck_01, loc=<4.5627 0 0>, rot=<0 0 0.0504 -0.9987>}
 head {parent=neck_02, loc=<5.3914 0 0>, rot=<0 0 0.115 -0.9934>}
 AD_head_frontQ_01 {parent=head, loc=<13.8959 11.8203 0.8295>, rot=<-0.0716 0.4117 0.8964 0.1482>}
 AD_head_frontR_01 {parent=head, loc=<12.7755 8.6005 -8.4635>, rot=<-0.2625 0.5675 0.711 0.3218>}
 AD_head_frontP_01 {parent=head, loc=<12.5601 7.4851 10.06>, rot=<0.3059 -0.5866 0.6677 0.3413>}
 AD_face_A1 {parent=head, loc=<-1.5516 9.7541 12.2975>, rot=<0.4082 -0.4611 0.7846 -0.0728>}
 CR_head_spikes_l {parent=head, loc=<1.1131 -8.1453 -6.9463>, rot=<0.9397 0 0 0.342>}
 AD_headTp_A1 {parent=CR_head_spikes_l, loc=<10.4834 -3.3639 0.2285>, rot=<0.4113 -0.5495 0.6444 0.3371>}
 AD_headTp_B1 {parent=CR_head_spikes_l, loc=<6.8881 1.4808 3.879>, rot=<0.3508 -0.5562 0.7404 0.1391>}
 AD_Sm_headTp_A1 {parent=CR_head_spikes_l, loc=<10.4834 -3.3639 0.2285>, rot=<0.4113 -0.5495 0.6444 0.3371>}
 AD_Sm_headTp_B1 {parent=CR_head_spikes_l, loc=<6.8881 1.4808 3.879>, rot=<0.3508 -0.5562 0.7404 0.1391>}
 CR_head_spikes_ct {parent=head, loc=<4.4892 -9.0797 0>, rot=<0 0 0 1>}
 AD_headTp_C1 {parent=CR_head_spikes_ct, loc=<2.1369 -5.1013 -4.6386>, rot=<0.2681 -0.8537 -0.3467 -0.2812>}
 AD_headTp_D1 {parent=CR_head_spikes_ct, loc=<1.9634 -6.348 0.1253>, rot=<0.2848 -0.9226 -0.0023 -0.2602>}
 AD_headTp_E1 {parent=CR_head_spikes_ct, loc=<1.746 -4.9354 4.8201>, rot=<0.3878 -0.8777 0.2133 0.1838>}
 AD_Med_headTp_C1 {parent=CR_head_spikes_ct, loc=<2.1369 -5.1013 -4.6386>, rot=<0.2681 -0.8537 -0.3467 -0.2812>}
 AD_Sm_headTp_D1 {parent=CR_head_spikes_ct, loc=<1.9634 -6.348 0.1253>, rot=<0.2848 -0.9226 -0.0023 -0.2602>}
 AD_Sm_headTp_E1 {parent=CR_head_spikes_ct, loc=<1.746 -4.9354 4.8201>, rot=<0.3878 -0.8777 0.2133 0.1838>}
 CR_head_spikes_ct1 {parent=head, loc=<-0.847 -9.1716 0>, rot=<0 0 0 1>}
 CR_headBa_D {parent=CR_head_spikes_ct1, loc=<2.7211 -6.0621 2.8588>, rot=<0.3482 -0.9259 -0.0192 -0.1451>}
 AD_headMid_D1 {parent=CR_headBa_D, loc=<0 0 0>, rot=<0 0 0 1>}
 AD_Med_headMid_D1 {parent=CR_headBa_D, loc=<0 0 0>, rot=<0 0 0 1>}
 CR_headBa_K {parent=CR_head_spikes_ct1, loc=<-0.9366 -6.071 -0.2969>, rot=<0.3134 -0.9354 -0.1135 -0.1179>}
 AD_headMid_K1 {parent=CR_headBa_K, loc=<0 0 0>, rot=<0 0 0 1>}
 AD_Med_headMid_K1 {parent=CR_headBa_K, loc=<0 0 0>, rot=<0 0 0 1>}
 CR_headBa_E {parent=CR_head_spikes_ct1, loc=<-0.702 -6.1181 6.8326>, rot=<0.2985 -0.8788 0.2442 0.281>}
 AD_headMid_E1 {parent=CR_headBa_E, loc=<0 0 0>, rot=<0 0 0 1>}
 AD_Med_headMid_E1 {parent=CR_headBa_E, loc=<0 0 0>, rot=<0 0 0 1>}
 CR_headBa_L {parent=CR_head_spikes_ct1, loc=<-3.452 -6.8498 5.1893>, rot=<0.3235 -0.9187 0.2254 -0.0235>}
 AD_headMid_L1 {parent=CR_headBa_L, loc=<0 0 0>, rot=<0 0 0 1>}
 AD_Lg_headMid_L1 {parent=CR_headBa_L, loc=<0 0 0>, rot=<0 0 0 1>}
 CR_headBa_J {parent=CR_head_spikes_ct1, loc=<0.0207 -5.7985 -4.6757>, rot=<0.3326 -0.8581 -0.3173 -0.2288>}
 AD_headMid_J1 {parent=CR_headBa_J, loc=<0 0 0>, rot=<0 0 0 1>}
 AD_Lg_headMid_J1 {parent=CR_headBa_J, loc=<0 0 0>, rot=<0 0 0 1>}
 CR_head_spikes_r1 {parent=head, loc=<-0.5789 -6.1617 8.6107>, rot=<-0.342 0 0 -0.9397>}
 CR_headMid_G {parent=CR_head_spikes_r1, loc=<0.507 -0.7467 6.6996>, rot=<0.2718 -0.6326 0.6025 0.4037>}
 AD_headMid_G1 {parent=CR_headMid_G, loc=<0 0 0>, rot=<0 0 0 1>}
 AD_Sm_headMid_G1 {parent=CR_headMid_G, loc=<0 0 0>, rot=<0 0 0 1>}
 CR_headMid_F {parent=CR_head_spikes_r1, loc=<2.3119 -5.5818 3.613>, rot=<0.1462 -0.6754 0.6109 0.3863>}
 AD_headMid_F1 {parent=CR_headMid_F, loc=<0 0 0>, rot=<0 0 0 1>}
 AD_Sm_headMid_F1 {parent=CR_headMid_F, loc=<0 0 0>, rot=<0 0 0 1>}
 AD_headTp_F1 {parent=CR_head_spikes_r1, loc=<5.8722 -2.5879 5.2894>, rot=<0.2703 -0.5538 0.6484 0.447>}
 AD_Med_headTp_F1 {parent=CR_head_spikes_r1, loc=<5.8722 -2.5879 5.2894>, rot=<0.2703 -0.5538 0.6484 0.447>}
 CR_head_spikes_r4 {parent=head, loc=<1.4452 -0.6094 7.6317>, rot=<0.0872 0 0 -0.9962>}
 AD_head_l_Q1 {parent=CR_head_spikes_r4, loc=<1.1714 -1.6079 8.1487>, rot=<0.3665 -0.7915 0.4767 0.1094>}
 AD_headTp_G1 {parent=CR_head_spikes_r4, loc=<3.2284 -6.5861 6.7594>, rot=<0.3427 -0.7923 0.4905 0.1196>}
 CR_head_l_P {parent=CR_head_spikes_r4, loc=<-1.9691 -3.9769 7.423>, rot=<0.4322 -0.8188 0.3549 0.1298>}
 AD_headMid_P1 {parent=CR_head_l_P, loc=<0 0 0>, rot=<0 0 0 1>}
 AD_Lg_headMid_P1 {parent=CR_head_l_P, loc=<0 0 0>, rot=<0 0 0 1>}
 AD_Lg_head_l_Q1 {parent=CR_head_spikes_r4, loc=<1.1714 -1.6079 8.1487>, rot=<0.3665 -0.7915 0.4767 0.1094>}
 AD_Lg_headTp_G1 {parent=CR_head_spikes_r4, loc=<3.2284 -6.5861 6.7594>, rot=<0.3427 -0.7923 0.4905 0.1196>}
 CR_head_spikes_l1 {parent=head, loc=<-0.5789 -6.1617 -8.6107>, rot=<0.9397 0 0 0.342>}
 CR_head_l_A {parent=CR_head_spikes_l1, loc=<1.8249 -9.5279 3.4013>, rot=<0.1726 -0.7801 0.4658 0.3803>}
 AD_headMid_A1 {parent=CR_head_l_A, loc=<0 0 0>, rot=<0 0 0 1>}
 AD_Lg_headMid_A1 {parent=CR_head_l_A, loc=<0 0 0>, rot=<0 0 0 1>}
 CR_headMid_C {parent=CR_head_spikes_l1, loc=<2.1911 2.6494 5.5859>, rot=<0.2773 -0.5328 0.7547 0.2637>}
 AD_headMid_C1 {parent=CR_headMid_C, loc=<0 0 0>, rot=<0 0 0 1>}
 AD_Lg_headMid_C1 {parent=CR_headMid_C, loc=<0 0 0>, rot=<0 0 0 1>}
 CR_head_l_B {parent=CR_head_spikes_l1, loc=<4.5306 -4.122 6.7039>, rot=<0.2815 -0.7609 0.5124 0.2815>}
 AD_headMid_B1 {parent=CR_head_l_B, loc=<0 0 0>, rot=<0 0 0 1>}
 AD_Lg_headMid_B1 {parent=CR_head_l_B, loc=<0 0 0>, rot=<0 0 0 1>}
 CR_head_l_H {parent=CR_head_spikes_l1, loc=<-2.6897 -2.7721 6.5648>, rot=<0.2669 -0.7153 0.5837 0.2762>}
 AD_headMid_H1 {parent=CR_head_l_H, loc=<0 0 0>, rot=<0 0 0 1>}
 AD_Med_headMid_H1 {parent=CR_head_l_H, loc=<0 0 0>, rot=<0 0 0 1>}
 CR_headBa_I {parent=CR_head_spikes_l1, loc=<-3.558 6.6094 5.9417>, rot=<0.3217 -0.5711 0.7019 0.2787>}
 AD_headMid_I1 {parent=CR_headBa_I, loc=<0 0 0>, rot=<0 0 0 1>}
 AD_Lg_headMid_I1 {parent=CR_headBa_I, loc=<0 0 0>, rot=<0 0 0 1>}
 CR_head_spike_fr_r {parent=head, loc=<-5.6317 5.5802 7.5311>, rot=<0.4226 0 0 -0.9063>}
 AD_face_C1 {parent=CR_head_spike_fr_r, loc=<-0.3711 5.1592 5.1143>, rot=<0.4576 -0.5416 0.7051 -0.0118>}
 AD_face_B1 {parent=CR_head_spike_fr_r, loc=<0.46 0.947 6.3021>, rot=<0.4714 -0.6418 0.595 -0.1091>}
 AD_Med_face_C1 {parent=CR_head_spike_fr_r, loc=<-0.3711 5.1592 5.1143>, rot=<0.4576 -0.5416 0.7051 -0.0118>}
 AD_Med_face_B1 {parent=CR_head_spike_fr_r, loc=<0.46 0.947 6.3021>, rot=<0.4714 -0.6418 0.595 -0.1091>}
 CR_head_spike_fr_ct {parent=head, loc=<-5.6317 5.5802 0>, rot=<0 0 0 1>}
 AD_face_D1 {parent=CR_head_spike_fr_ct, loc=<-1.0639 8.3862 3.3319>, rot=<0.2847 -0.2389 0.88 0.2959>}
 AD_face_G1 {parent=CR_head_spike_fr_ct, loc=<-0.1625 8.4382 -1.3249>, rot=<-0.1215 0.054 0.9676 0.2147>}
 AD_face_H1 {parent=CR_head_spike_fr_ct, loc=<-0.1339 7.4206 -5.5868>, rot=<-0.2774 0.1987 0.9176 0.2039>}
 AD_Med_face_D1 {parent=CR_head_spike_fr_ct, loc=<-1.0639 8.3862 3.3319>, rot=<0.2847 -0.2389 0.88 0.2959>}
 AD_Med_face_G1 {parent=CR_head_spike_fr_ct, loc=<-0.1625 8.4382 -1.3249>, rot=<-0.1215 0.054 0.9676 0.2147>}
 AD_Med_face_H1 {parent=CR_head_spike_fr_ct, loc=<-0.1339 7.4206 -5.5868>, rot=<-0.2774 0.1987 0.9176 0.2039>}
 CR_head_spike_fr_l {parent=head, loc=<-5.6317 5.5802 -7.0339>, rot=<0.9397 0 0 0.342>}
 AD_face_I1 {parent=CR_head_spike_fr_l, loc=<-0.7038 -5.9184 -0.1248>, rot=<0.3707 -0.8991 -0.0675 0.2229>}
 AD_face_J1 {parent=CR_head_spike_fr_l, loc=<1.8765 -6.4322 3.4854>, rot=<0.2475 -0.8781 0.1246 0.3902>}
 AD_Sm_face_I1 {parent=CR_head_spike_fr_l, loc=<-0.7038 -5.9184 -0.1248>, rot=<0.3707 -0.8991 -0.0675 0.2229>}
 AD_Sm_face_J1 {parent=CR_head_spike_fr_l, loc=<1.8765 -6.4322 3.4854>, rot=<0.2475 -0.8781 0.1246 0.3902>}
 CR_head_spike_fr_l2 {parent=head, loc=<-2.3095 2.3321 -12.3014>, rot=<-0.8192 0 0 -0.5736>}
 CR_face_l_J {parent=CR_head_spike_fr_l2, loc=<-0.72 -3.9163 1.5153>, rot=<0.1985 -0.9215 -0.0067 0.3338>}
 AD_faceBt_J1 {parent=CR_face_l_J, loc=<0 0 0>, rot=<0 0 0 1>}
 AD_Med_faceBt_J1 {parent=CR_face_l_J, loc=<0 0 0>, rot=<0 0 0 1>}
 CR_face_l_I {parent=CR_head_spike_fr_l2, loc=<-4.8427 -4.9195 -0.6929>, rot=<0.2977 -0.8518 -0.138 0.4084>}
 AD_faceBt_I1 {parent=CR_face_l_I, loc=<0 0 0>, rot=<0 0 0 1>}
 AD_Med_faceBt_I1 {parent=CR_face_l_I, loc=<0 0 0>, rot=<0 0 0 1>}
 CR_head_spike_fr_r2 {parent=head, loc=<-7.9905 3.5423 9.9598>, rot=<0.2588 0 0 -0.9659>}
 CR_faceBt_B {parent=CR_head_spike_fr_r2, loc=<-2.4395 3.1573 3.6393>, rot=<0.4381 -0.572 0.6889 0.0792>}
 AD_faceBt_B1 {parent=CR_faceBt_B, loc=<0 0 0>, rot=<0 0 0 1>}
 AD_Sm_faceBt_B1 {parent=CR_faceBt_B, loc=<0 0 0>, rot=<0 0 0 1>}
 CR_face_l_M {parent=CR_head_spike_fr_r2, loc=<5.1486 -3.6297 5.7647>, rot=<0.345 -0.7968 0.4955 -0.0222>}
 AD_faceBt_M1 {parent=CR_face_l_M, loc=<0 0 0>, rot=<0 0 0 1>}
 AD_Med_faceBt_M1 {parent=CR_face_l_M, loc=<0 0 0>, rot=<0 0 0 1>}
 CR_faceBt_A {parent=CR_head_spike_fr_r2, loc=<0.8287 0.5089 4.6785>, rot=<0.5317 -0.6149 0.5768 -0.0808>}
 AD_faceBt_A1 {parent=CR_faceBt_A, loc=<0 0 0>, rot=<0 0 0 1>}
 AD_Sm_faceBt_A1 {parent=CR_faceBt_A, loc=<0 0 0>, rot=<0 0 0 1>}
 AD_Sm_head_frontQ_01 {parent=head, loc=<13.8959 11.8203 0.8295>, rot=<-0.0716 0.4117 0.8964 0.1482>}
 AD_Med_head_frontR_01 {parent=head, loc=<12.7755 8.6005 -8.4635>, rot=<-0.2625 0.5675 0.711 0.3218>}
 AD_Med_head_frontP_01 {parent=head, loc=<12.5601 7.4851 10.06>, rot=<0.3059 -0.5866 0.6677 0.3413>}
 AD_Sm_face_A1 {parent=head, loc=<-1.5516 9.7541 12.2975>, rot=<0.4082 -0.4611 0.7846 -0.0728>}
 CR_head_spikes_r2 {parent=neck_02, loc=<-4.5427 -3.9986 8.6107>, rot=<-0.3398 0.0393 0.108 -0.9335>}
 CR_headBa_M {parent=CR_head_spikes_r2, loc=<2.9785 -6.6986 3.6404>, rot=<0.2238 -0.7591 0.5823 0.1857>}
 AD_headMid_M1 {parent=CR_headBa_M, loc=<0 0 0>, rot=<0 0 0 1>}
 AD_Sm_headMid_M1 {parent=CR_headBa_M, loc=<0 0 0>, rot=<0 0 0 1>}
 CR_headBa_N {parent=CR_head_spikes_r2, loc=<4.4534 -4.7092 4.3408>, rot=<0.246 -0.6926 0.563 0.378>}
 AD_headMid_N1 {parent=CR_headBa_N, loc=<0 0 0>, rot=<0 0 0 1>}
 AD_Med_headMid_N1 {parent=CR_headBa_N, loc=<0 0 0>, rot=<0 0 0 1>}
 CR_head_l_O {parent=CR_head_spikes_r2, loc=<6.9179 0.8965 3.4856>, rot=<0.2049 -0.6906 0.5944 0.3575>}
 AD_headMid_O1 {parent=CR_head_l_O, loc=<0 0 0>, rot=<0 0 0 1>}
 AD_Med_headMid_O1 {parent=CR_head_l_O, loc=<0 0 0>, rot=<0 0 0 1>}
 CR_head_spike_fr_ct1 {parent=neck_02, loc=<-5.3943 8.2619 0>, rot=<0 0 0.115 -0.9934>}
 CR_faceBt_D {parent=CR_head_spike_fr_ct1, loc=<2.1616 6.9929 1.2247>, rot=<0.1308 -0.056 0.9706 0.194>}
 AD_faceBt_D1 {parent=CR_faceBt_D, loc=<0 0 0>, rot=<0 0 0 1>}
 AD_Sm_faceBt_D1 {parent=CR_faceBt_D, loc=<0 0 0>, rot=<0 0 0 1>}
 CR_faceBt_F {parent=CR_head_spike_fr_ct1, loc=<2.5913 6.6669 -4.2706>, rot=<-0.1952 0.1452 0.9605 0.135>}
 AD_faceBt_F1 {parent=CR_faceBt_F, loc=<0 0 0>, rot=<0 0 0 1>}
 AD_Sm_faceBt_F1 {parent=CR_faceBt_F, loc=<0 0 0>, rot=<0 0 0 1>}
 CR_faceBt_C {parent=CR_head_spike_fr_ct1, loc=<2.0894 6.0387 7.0319>, rot=<0.2286 -0.2554 0.9387 0.0364>}
 AD_faceBt_C1 {parent=CR_faceBt_C, loc=<0 0 0>, rot=<0 0 0 1>}
 AD_Lg_faceBt_C1 {parent=CR_faceBt_C, loc=<0 0 0>, rot=<0 0 0 1>}
 CR_head_spike_fr_r1 {parent=neck_01, loc=<-0.1427 8.267 8.1566>, rot=<0.4168 -0.0697 0.1495 -0.8939>}
 CR_headFr_B_bt {parent=CR_head_spike_fr_r1, loc=<-0.1232 -2.8332 3.4137>, rot=<0.3064 -0.749 0.5719 0.1348>}
 AD_headFr_B_bt1 {parent=CR_headFr_B_bt, loc=<0 0 0>, rot=<0 0 0 1>}
 AD_Lg_headFr_B_bt1 {parent=CR_headFr_B_bt, loc=<0 0 0>, rot=<0 0 0 1>}
 CR_headFr_C_bt {parent=CR_head_spike_fr_r1, loc=<-0.0482 2.0786 3.1936>, rot=<0.3528 -0.8071 0.4287 0.2008>}
 AD_headFr_C_bt1 {parent=CR_headFr_C_bt, loc=<0 0 0>, rot=<0 0 0 1>}
 AD_Med_headFr_C_bt1 {parent=CR_headFr_C_bt, loc=<0 0 0>, rot=<0 0 0 1>}
 CR_head_spike_fr_l1 {parent=neck_01, loc=<0.0813 5.9214 -9.3002>, rot=<0.9268 -0.155 -0.0564 0.3373>}
 AD_faceBt_H1 {parent=CR_head_spike_fr_l1, loc=<4.7101 -6.4233 -0.6081>, rot=<0.2739 -0.9219 -0.0174 0.2734>}
 CR_faceBt_G {parent=CR_head_spike_fr_l1, loc=<2.7316 -5.6694 -5.461>, rot=<0.2829 -0.92 -0.1196 0.2434>}
 AD_faceBt_G1 {parent=CR_faceBt_G, loc=<0 0 0>, rot=<0 0 0 1>}
 AD_Sm_faceBt_G1 {parent=CR_faceBt_G, loc=<0 0 0>, rot=<0 0 0 1>}
 AD_Sm_faceBt_H1 {parent=CR_head_spike_fr_l1, loc=<4.7101 -6.4233 -0.6081>, rot=<0.2739 -0.9219 -0.0174 0.2734>}
 attach_backpack {parent=spine_05, loc=<-0.123 -13.473 0>, rot=<-0.0093 -0.707 0.0093 -0.707>}
 clavicle_l {parent=spine_05, loc=<4.3575 -0.1466 -2.3223>, rot=<0.0465 -0.7289 -0.009 -0.6829>}
 upperarm_l {parent=clavicle_l, loc=<15.5111 0 0>, rot=<0.0226 -0.3234 -0.0088 -0.9459>}
 lowerarm_l {parent=upperarm_l, loc=<26.6514 0 0>, rot=<0.0391 -0.0099 -0.2453 -0.9686>}
 hand_l {parent=lowerarm_l, loc=<22.3491 0 0>, rot=<-0.7081 -0.0507 0.0975 -0.6975>}
 thumb_01_l {parent=hand_l, loc=<1.3118 -0.3191 -2.459>, rot=<0.6546 -0.4042 0.0664 -0.6354>}
 thumb_02_l {parent=thumb_01_l, loc=<5.9336 -0.0448 0.0401>, rot=<-0.0037 0.0213 0.298 -0.9543>}
 thumb_03_l {parent=thumb_02_l, loc=<2.2223 0.3682 -0.0907>, rot=<0.0118 -0.013 0.0066 -0.9998>}
 index_metacarpal_l {parent=hand_l, loc=<1.6535 0.0183 -2.96>, rot=<0.0158 -0.0124 -0.0173 -0.9996>}
 index_01_l {parent=index_metacarpal_l, loc=<8.3008 0.0357 -0.0524>, rot=<-0.0044 0.0485 0.0924 -0.9945>}
 index_02_l {parent=index_01_l, loc=<3.3587 -0.3004 0.0452>, rot=<-0.002 0.0027 0.3225 -0.9466>}
 index_03_l {parent=index_02_l, loc=<2.5578 -0.123 -0.0252>, rot=<-0.002 0.0067 0.2558 -0.9667>}
 middle_metacarpal_l {parent=hand_l, loc=<2.1856 -0.1725 -0.5984>, rot=<0.006 0.0168 -0.0329 -0.9993>}
 middle_01_l {parent=middle_metacarpal_l, loc=<7.3275 -0.2447 0.0129>, rot=<-0.0208 0.0042 0.0858 -0.9961>}
 middle_02_l {parent=middle_01_l, loc=<4.0123 0.1401 0.1175>, rot=<-0.0218 0.0258 0.3512 -0.9357>}
 middle_03_l {parent=middle_02_l, loc=<2.6077 0.0534 0.1373>, rot=<0.0082 0.0188 0.1708 -0.9851>}
 ring_metacarpal_l {parent=hand_l, loc=<2.2074 -0.3625 1.3638>, rot=<-0.0284 0.0722 -0.0458 -0.9959>}
 ring_01_l {parent=ring_metacarpal_l, loc=<7.0295 -0.6197 -0.1446>, rot=<-0.0489 -0.0252 0.1892 -0.9804>}
 ring_02_l {parent=ring_01_l, loc=<3.6174 0.4647 -0.1027>, rot=<0 0 0.2806 -0.9598>}
 ring_03_l {parent=ring_02_l, loc=<2.4152 0.114 0.0199>, rot=<0 0 0.1189 -0.9929>}
 pinky_metacarpal_l {parent=hand_l, loc=<1.7616 -0.9753 3.3814>, rot=<-0.0397 0.0814 -0.069 -0.9935>}
 pinky_01_l {parent=pinky_metacarpal_l, loc=<7.1709 -0.4831 -0.1094>, rot=<-0.0654 -0.0035 0.2242 -0.9723>}
 pinky_02_l {parent=pinky_01_l, loc=<2.8274 0.0893 -0.1043>, rot=<0.0037 0.0011 0.2854 -0.9584>}
 pinky_03_l {parent=pinky_02_l, loc=<2.2046 0.0307 -0.0078>, rot=<-0.0002 0.0098 0.1445 -0.9895>}
 lowerarm_twist_02_l {parent=lowerarm_l, loc=<9.6199 0 0>, rot=<0 0 0 -1>}
 lowerarm_twist_01_l {parent=lowerarm_l, loc=<17.8549 0 0>, rot=<0 0 0 -1>}
 upperarm_twist_01_l {parent=upperarm_l, loc=<0.0041 0 0>, rot=<0 0 0 -1>}
 upperarm_twist_02_l {parent=upperarm_l, loc=<13.7941 0 0>, rot=<0 0 0 -1>}
 clavicle_r {parent=spine_05, loc=<4.3575 -0.1466 2.3223>, rot=<0.7289 0.0465 -0.6829 0.009>}
 upperarm_r {parent=clavicle_r, loc=<-15.5111 0 0>, rot=<0.0226 -0.3234 -0.0088 -0.9459>}
 lowerarm_r {parent=upperarm_r, loc=<-26.6514 0 0>, rot=<0.0391 -0.0099 -0.2453 -0.9686>}
 hand_r {parent=lowerarm_r, loc=<-22.3491 0 0>, rot=<-0.7081 -0.0507 0.0975 -0.6975>}
 thumb_01_r {parent=hand_r, loc=<-1.3118 0.3191 2.459>, rot=<0.6546 -0.4042 0.0664 -0.6354>}
 thumb_02_r {parent=thumb_01_r, loc=<-5.9336 0.0448 -0.0401>, rot=<-0.0037 0.0213 0.298 -0.9543>}
 thumb_03_r {parent=thumb_02_r, loc=<-2.2223 -0.3682 0.0907>, rot=<0.0118 -0.013 0.0066 -0.9998>}
 index_metacarpal_r {parent=hand_r, loc=<-1.6535 -0.0183 2.96>, rot=<0.0158 -0.0124 -0.0173 -0.9996>}
 index_01_r {parent=index_metacarpal_r, loc=<-8.3008 -0.0357 0.0524>, rot=<-0.0044 0.0485 0.0924 -0.9945>}
 index_02_r {parent=index_01_r, loc=<-3.3587 0.3004 -0.0452>, rot=<-0.002 0.0027 0.3225 -0.9466>}
 index_03_r {parent=index_02_r, loc=<-2.5578 0.123 0.0252>, rot=<-0.002 0.0067 0.2558 -0.9667>}
 middle_metacarpal_r {parent=hand_r, loc=<-2.1856 0.1725 0.5984>, rot=<0.006 0.0168 -0.0329 -0.9993>}
 middle_01_r {parent=middle_metacarpal_r, loc=<-7.3275 0.2447 -0.0129>, rot=<-0.0208 0.0042 0.0858 -0.9961>}
 middle_02_r {parent=middle_01_r, loc=<-4.0123 -0.1401 -0.1175>, rot=<-0.0218 0.0258 0.3512 -0.9357>}
 middle_03_r {parent=middle_02_r, loc=<-2.6077 -0.0534 -0.1373>, rot=<0.0082 0.0188 0.1708 -0.9851>}
 ring_metacarpal_r {parent=hand_r, loc=<-2.2074 0.3625 -1.3638>, rot=<-0.0284 0.0722 -0.0458 -0.9959>}
 ring_01_r {parent=ring_metacarpal_r, loc=<-7.0295 0.6197 0.1446>, rot=<-0.0489 -0.0252 0.1892 -0.9804>}
 ring_02_r {parent=ring_01_r, loc=<-3.6174 -0.4647 0.1027>, rot=<0 0 0.2806 -0.9598>}
 ring_03_r {parent=ring_02_r, loc=<-2.4152 -0.114 -0.0199>, rot=<0 0 0.1189 -0.9929>}
 pinky_metacarpal_r {parent=hand_r, loc=<-1.7616 0.9753 -3.3814>, rot=<-0.0397 0.0814 -0.069 -0.9935>}
 pinky_01_r {parent=pinky_metacarpal_r, loc=<-7.1709 0.4831 0.1094>, rot=<-0.0654 -0.0035 0.2242 -0.9723>}
 pinky_02_r {parent=pinky_01_r, loc=<-2.8274 -0.0893 0.1043>, rot=<0.0037 0.0011 0.2854 -0.9584>}
 pinky_03_r {parent=pinky_02_r, loc=<-2.2046 -0.0307 0.0078>, rot=<-0.0002 0.0098 0.1445 -0.9895>}
 lowerarm_twist_02_r {parent=lowerarm_r, loc=<-9.6199 0 0>, rot=<0 0 0 -1>}
 lowerarm_twist_01_r {parent=lowerarm_r, loc=<-17.8549 0 0>, rot=<0 0 0 -1>}
 upperarm_twist_01_r {parent=upperarm_r, loc=<-0.0041 0 0>, rot=<0 0 0 -1>}
 upperarm_twist_02_r {parent=upperarm_r, loc=<-13.7941 0 0>, rot=<0 0 0 -1>}
 CR_head_spikes_r3 {parent=clavicle_r, loc=<-8.6992 0.3638 -3.4887>, rot=<0.7301 0.0202 -0.6828 0.0156>}
 CR_shoulder_r_A {parent=CR_head_spikes_r3, loc=<2.6541 -7.0506 1.3911>, rot=<0.3746 -0.7523 0.3924 0.3738>}
 AD_shoulder_A_r1 {parent=CR_shoulder_r_A, loc=<0 0 0>, rot=<0 0 0 1>}
 AD_Lg_shoulder_A_r1 {parent=CR_shoulder_r_A, loc=<0 0 0>, rot=<0 0 0 1>}
 CR_shoulder_r_B {parent=CR_head_spikes_r3, loc=<5.0212 -3.7648 3.6859>, rot=<0.4378 -0.6682 0.4255 0.4252>}
 AD_shoulder_B_r1 {parent=CR_shoulder_r_B, loc=<0 0 0>, rot=<0 0 0 1>}
 AD_Lg_shoulder_B_r1 {parent=CR_shoulder_r_B, loc=<0 0 0>, rot=<0 0 0 1>}
 CR_shoulder_r_C {parent=CR_head_spikes_r3, loc=<6.2531 2.6054 2.0298>, rot=<0.4814 -0.5012 0.6287 0.349>}
 AD_shoulder_C_r1 {parent=CR_shoulder_r_C, loc=<0 0 0>, rot=<0 0 0 1>}
 AD_Lg_shoulder_C_r1 {parent=CR_shoulder_r_C, loc=<0 0 0>, rot=<0 0 0 1>}
 CR_shoulder_r_D {parent=CR_head_spikes_r3, loc=<2.751 5.6152 3.3281>, rot=<0.4602 -0.4133 0.7044 0.3482>}
 AD_shoulder_D_r1 {parent=CR_shoulder_r_D, loc=<0 0 0>, rot=<0 0 0 1>}
 AD_Sm_shoulder_D_r1 {parent=CR_shoulder_r_D, loc=<0 0 0>, rot=<0 0 0 1>}
 CR_headFr_A_bt {parent=CR_head_spikes_r3, loc=<1.0763 6.8967 2.7152>, rot=<0.3124 -0.4417 0.8085 0.2314>}
 AD_headFr_A_bt1 {parent=CR_headFr_A_bt, loc=<0 0 0>, rot=<0 0 0 1>}
 AD_Sm_headFr_A_bt1 {parent=CR_headFr_A_bt, loc=<0 0 0>, rot=<0 0 0 1>}
 AD_pine_rowA_L1 {parent=spine_05, loc=<8.1023 -6.0806 6.4281>, rot=<0.4223 -0.7571 0.4534 0.2071>}
 AD_pine_rowA_D1 {parent=spine_05, loc=<8.2376 -7.2531 -10.7265>, rot=<0.3392 -0.6497 -0.6264 -0.2655>}
 AD_pine_rowA_E1 {parent=spine_05, loc=<7.3992 -9.0765 -4.9878>, rot=<0.3448 -0.6661 -0.6098 -0.256>}
 AD_pine_rowA_H1 {parent=spine_05, loc=<7.9156 -7.5522 8.5355>, rot=<0.3609 -0.8527 -0.3585 -0.1189>}
 AD_pine_rowA_G1 {parent=spine_05, loc=<7.7065 -8.2236 4.7461>, rot=<0.3853 -0.8412 -0.3565 -0.1298>}
 AD_pine_rowA_O1 {parent=spine_05, loc=<9.6625 6.2841 10.3494>, rot=<0.1932 -0.5942 0.7595 0.1812>}
 AD_pine_rowA_P1 {parent=spine_05, loc=<9.1047 11.0455 6.2756>, rot=<0.1573 -0.5437 0.8078 0.1649>}
 AD_pine_rowA_B1 {parent=spine_05, loc=<9.0093 12.2108 3.4664>, rot=<-0.0362 0.2297 0.9698 0.0741>}
 AD_pine_rowA_T1 {parent=spine_05, loc=<7.0821 9.8179 -4.6303>, rot=<-0.2588 0.527 0.7529 0.2974>}
 CR_spineRot_05 {parent=spine_05, loc=<0 0 0>, rot=<0 0 0 1>}
 CR_headBa_A_bt {parent=CR_spineRot_05, loc=<12.2891 -8.7178 -7.1994>, rot=<0.1948 -0.9296 -0.1416 -0.2789>}
 AD_headBa_A_bt1 {parent=CR_headBa_A_bt, loc=<0 0 0>, rot=<0 0 0 1>}
 CR_head_spike_fr_ct2 {parent=spine_05, loc=<5.7389 8.0683 0>, rot=<0 0 -0.036 -0.9994>}
 CR_headFr_D_bt {parent=CR_head_spike_fr_ct2, loc=<2.4213 4.8614 3.4864>, rot=<0.2433 -0.4237 0.8278 0.2759>}
 AD_headFr_D_bt1 {parent=CR_headFr_D_bt, loc=<0 0 0>, rot=<0 0 0 1>}
 CR_headFr_E_bt {parent=CR_head_spike_fr_ct2, loc=<-2.6987 6.024 4.4188>, rot=<-0.1784 -0.2057 0.9187 0.2862>}
 AD_headFr_E_bt1 {parent=CR_headFr_E_bt, loc=<-7.7612 0 0>, rot=<0 0 0 1>}
 CR_headFr_F_bt {parent=CR_head_spike_fr_ct2, loc=<3.6219 3.6981 -2.1664>, rot=<-0.1845 0.1455 0.9621 0.1384>}
 AD_headFr_F_bt1 {parent=CR_headFr_F_bt, loc=<0 0 0>, rot=<0 0 0 1>}
 CR_faceBt_E {parent=CR_head_spike_fr_ct2, loc=<1.4614 6.2327 -0.7442>, rot=<-0.11 -0.0225 0.9751 0.1911>}
 AD_faceBt_E1 {parent=CR_faceBt_E, loc=<0 0 0>, rot=<0 0 0 1>}
 CR_head_spikes_ct2 {parent=spine_05, loc=<12.5392 -5.6488 0>, rot=<0 0 -0.036 -0.9994>}
 CR_headBa_B_bt {parent=CR_head_spikes_ct2, loc=<1.6607 -4.8156 -1.885>, rot=<0.2393 -0.8771 -0.1726 -0.3791>}
 AD_headBa_B_bt1 {parent=CR_headBa_B_bt, loc=<0 0 0>, rot=<0 0 0 1>}
 CR_headBa_C_bt {parent=CR_head_spikes_ct2, loc=<-2.4122 -5.4121 0.3303>, rot=<0.1192 -0.9361 0.0391 -0.3287>}
 AD_headBa_C_bt1 {parent=CR_headBa_C_bt, loc=<0 0 0>, rot=<0 0 0 1>}
 CR_headBa_D_bt {parent=CR_head_spikes_ct2, loc=<-3.0335 -6.437 3.3511>, rot=<0.1759 -0.9731 0.1019 -0.1084>}
 AD_headBa_D_bt1 {parent=CR_headBa_D_bt, loc=<0 0 0>, rot=<0 0 0 1>}
 CR_headBa_E_bt {parent=CR_head_spikes_ct2, loc=<-7.9323 -5.9113 5.8524>, rot=<0.3753 -0.7888 0.3586 0.3294>}
 AD_headBa_E_bt1 {parent=CR_headBa_E_bt, loc=<0 0 0>, rot=<0 0 0 1>}
 CR_headBa_F_bt {parent=CR_head_spikes_ct2, loc=<-3.8485 -4.2451 6.8999>, rot=<0.3275 -0.8678 0.2231 0.2998>}
 AD_headBa_F_bt1 {parent=CR_headBa_F_bt, loc=<0 0 0>, rot=<0 0 0 1>}
 CR_head_spikes_l3 {parent=spine_05, loc=<7.3105 -0.7241 -11.2063>, rot=<0.9994 0.036 0 0>}
 CR_shoulder_l_B {parent=CR_head_spikes_l3, loc=<5.3799 4.2908 0.6939>, rot=<0.4332 -0.4093 0.7472 0.2942>}
 AD_shoulder_B_l1 {parent=CR_shoulder_l_B, loc=<0 0 0>, rot=<0 0 0 1>}
 AD_Lg_shoulder_B_l1 {parent=CR_shoulder_l_B, loc=<0 0 0>, rot=<0 0 0 1>}
 CR_shoulder_l_C {parent=CR_head_spikes_l3, loc=<5.6069 -1.575 2.9613>, rot=<0.337 -0.6539 0.544 0.4036>}
 AD_shoulder_C_l1 {parent=CR_shoulder_l_C, loc=<0 0 0>, rot=<0 0 0 1>}
 AD_Lg_shoulder_C_l1 {parent=CR_shoulder_l_C, loc=<0 0 0>, rot=<0 0 0 1>}
 CR_shoulder_l_D {parent=CR_head_spikes_l3, loc=<4.6604 -7.1671 0.8322>, rot=<0.1967 -0.8088 0.3706 0.4121>}
 AD_shoulder_D_l1 {parent=CR_shoulder_l_D, loc=<0 0 0>, rot=<0 0 0 1>}
 AD_Lg_shoulder_D_l1 {parent=CR_shoulder_l_D, loc=<0 0 0>, rot=<0 0 0 1>}
 CR_shoulder_l_E {parent=CR_head_spikes_l3, loc=<3.7562 -9.5793 -1.8126>, rot=<0.1641 -0.8948 0.3021 0.2848>}
 AD_shoulder_E_l1 {parent=CR_shoulder_l_E, loc=<0 0 0>, rot=<0 0 0 1>}
 AD_Lg_shoulder_E_l1 {parent=CR_shoulder_l_E, loc=<0 0 0>, rot=<0 0 0 1>}
 CR_shoulder_l_A {parent=CR_head_spikes_l3, loc=<-0.2646 6.6941 3.6751>, rot=<0.4535 -0.319 0.8211 0.1353>}
 AD_shoulder_A_l1 {parent=CR_shoulder_l_A, loc=<0 0 0>, rot=<0 0 0 1>}
 AD_Lg_shoulder_A_l1 {parent=CR_shoulder_l_A, loc=<0 0 0>, rot=<0 0 0 1>}
 AD_Sm_pine_rowA_L1 {parent=spine_05, loc=<3.7747 -9.5857 11.8339>, rot=<0.4223 -0.7571 0.4534 0.2071>}
 AD_Sm_pine_rowA_D1 {parent=spine_05, loc=<8.2376 -7.2531 -10.7265>, rot=<0.3392 -0.6497 -0.6264 -0.2655>}
 AD_Sm_pine_rowA_E1 {parent=spine_05, loc=<2.5003 -10.2185 -10.8982>, rot=<0.3448 -0.6661 -0.6098 -0.256>}
 AD_Sm_pine_rowA_H1 {parent=spine_05, loc=<2.3953 -11.6676 4.9541>, rot=<0.3609 -0.8527 -0.3585 -0.1189>}
 AD_Sm_pine_rowA_G1 {parent=spine_05, loc=<2.5113 -12.5361 0.9188>, rot=<0.3853 -0.8412 -0.3565 -0.1298>}
 AD_Sm_pine_rowA_O1 {parent=spine_05, loc=<2.99 6.6378 14.2977>, rot=<0.1932 -0.5942 0.7595 0.1812>}
 AD_Sm_pine_rowA_P1 {parent=spine_05, loc=<2.1498 11.7861 9.6397>, rot=<0.1573 -0.5437 0.8078 0.1649>}
 AD_Sm_pine_rowA_B1 {parent=spine_05, loc=<1.3536 13.1964 2.6568>, rot=<-0.0362 0.2297 0.9698 0.0741>}
 AD_Sm_pine_rowA_T1 {parent=spine_05, loc=<1.7329 11.1763 -10.0872>, rot=<-0.2588 0.527 0.7529 0.2974>}
 AD_pine_rowA_F1 {parent=spine_04, loc=<12.152 -8.4785 -4.2582>, rot=<0.3706 -0.7916 -0.4557 -0.1687>}
 AD_pine_rowA_I1 {parent=spine_04, loc=<12.0288 -8.8756 11.3431>, rot=<0.3874 -0.8128 -0.4069 -0.1539>}
 AD_pine_rowA_J1 {parent=spine_04, loc=<16.009 -9.1258 4.3723>, rot=<0.2524 -0.8601 0.4343 0.089>}
 AD_pine_rowA_K1 {parent=spine_04, loc=<15.4775 -9.9882 5.5061>, rot=<0.2594 -0.7144 0.6272 0.1702>}
 AD_pine_rowA_N1 {parent=spine_04, loc=<13.0349 -0.2329 13.4784>, rot=<0.2409 -0.7551 0.5943 0.1364>}
 AD_pine_rowA_M1 {parent=spine_04, loc=<12.4141 -9.0932 9.1348>, rot=<0.3562 -0.5795 0.6542 0.3306>}
 AD_pine_rowB_C1 {parent=spine_04, loc=<9.79 7.1984 11.2433>, rot=<0.2512 -0.6747 0.6682 0.1873>}
 AD_pine_rowB_Q1 {parent=spine_04, loc=<7.8132 -9.2548 8.9479>, rot=<0.3021 -0.5997 0.6872 0.2772>}
 AD_pine_rowB_K1 {parent=spine_04, loc=<7.9466 -9.4711 0.057>, rot=<0.39 -0.834 -0.3661 -0.1354>}
 AD_pine_rowB_I1 {parent=spine_04, loc=<7.4974 -5.3195 -11.0621>, rot=<0.3733 -0.7862 -0.4609 -0.1735>}
 AD_pine_rowB_G1 {parent=spine_04, loc=<8.4832 1.3473 -13.3372>, rot=<-0.2363 0.605 0.7288 0.2168>}
 AD_pine_rowB_J1 {parent=spine_04, loc=<7.9218 -9.9094 -3.8496>, rot=<0.274 -0.8299 -0.4725 -0.1135>}
 AD_pine_rowB_N1 {parent=spine_04, loc=<6.2007 -8.9613 4.109>, rot=<0.4018 -0.8186 0.3822 0.1496>}
 AD_pine_rowB_R1 {parent=spine_04, loc=<6.8936 -8.4656 9.1994>, rot=<0.2484 -0.7119 0.636 0.1641>}
 AD_pine_rowA_C1 {parent=spine_04, loc=<11.3607 11.2756 -4.779>, rot=<-0.2163 0.4376 0.8119 0.3203>}
 AD_pine_rowA_A1 {parent=spine_04, loc=<14.7346 11.3092 0.1149>, rot=<-0.1415 0.3903 0.8781 0.2379>}
 AD_pine_rowB_A1 {parent=spine_04, loc=<11.5368 11.1161 8.8186>, rot=<0.2518 -0.6517 0.6865 0.2015>}
 AD_pine_rowB_B1 {parent=spine_04, loc=<7.7701 12.1031 6.2528>, rot=<0.2655 -0.6597 0.6718 0.2074>}
 AD_pine_rowB_F1 {parent=spine_04, loc=<8.1027 11.9167 -9.1063>, rot=<-0.1309 0.5907 0.7889 0.1071>}
 AD_pine_rowB_E1 {parent=spine_04, loc=<5.1986 12.588 0.9815>, rot=<-0.258 0.5166 0.7582 0.3028>}
 AD_pine_rowB_D1 {parent=spine_04, loc=<6.4622 11.2048 2.601>, rot=<0.0599 -0.2113 0.9574 0.1872>}
 AD_pine_rowA_Q1 {parent=spine_04, loc=<13.4318 11.3468 2.5217>, rot=<0.2143 -0.5226 0.7869 0.2485>}
 AD_Sm_pine_rowA_F1 {parent=spine_04, loc=<6.9644 -11.8382 -8.9526>, rot=<0.3706 -0.7916 -0.4557 -0.1687>}
 AD_Sm_pine_rowA_I1 {parent=spine_04, loc=<6.9652 -12.7916 6.9544>, rot=<0.3874 -0.8128 -0.4069 -0.1539>}
 AD_Sm_pine_rowA_J1 {parent=spine_04, loc=<9.3596 -11.8954 7.2622>, rot=<0.2524 -0.8601 0.4343 0.089>}
 AD_Sm_pine_rowA_K1 {parent=spine_04, loc=<9.2104 -11.2073 9.919>, rot=<0.2594 -0.7144 0.6272 0.1702>}
 AD_Sm_pine_rowA_N1 {parent=spine_04, loc=<6.463 -1.7985 17.2989>, rot=<0.2409 -0.7551 0.5943 0.1364>}
 AD_Sm_pine_rowA_M1 {parent=spine_04, loc=<8.3184 -8.9405 15.7255>, rot=<0.3562 -0.5795 0.6542 0.3306>}
 AD_Lg_pine_rowB_C1 {parent=spine_04, loc=<3.5529 6.5094 15.8108>, rot=<0.2512 -0.6747 0.6682 0.1873>}
 AD_Med_pine_rowB_Q1 {parent=spine_04, loc=<2.6615 -9.1099 14.751>, rot=<0.3021 -0.5997 0.6872 0.2772>}
 AD_Med_pine_rowB_K1 {parent=spine_04, loc=<2.8307 -13.7505 -3.9118>, rot=<0.39 -0.834 -0.3661 -0.1354>}
 AD_Lg_pine_rowB_I1 {parent=spine_04, loc=<2.3664 -8.6331 -15.8506>, rot=<0.3733 -0.7862 -0.4609 -0.1735>}
 AD_Lg_pine_rowB_G1 {parent=spine_04, loc=<2.3177 1.5809 -18.0456>, rot=<-0.2363 0.605 0.7288 0.2168>}
 AD_Med_pine_rowB_J1 {parent=spine_04, loc=<1.5255 -12.6066 -7.3207>, rot=<0.274 -0.8299 -0.4725 -0.1135>}
 AD_Sm_pine_rowB_N1 {parent=spine_04, loc=<1.2927 -13.179 8.3936>, rot=<0.4018 -0.8186 0.3822 0.1496>}
 AD_Sm_pine_rowB_R1 {parent=spine_04, loc=<0.5087 -9.5906 13.466>, rot=<0.2484 -0.7119 0.636 0.1641>}
 AD_Sm_pine_rowA_C1 {parent=spine_04, loc=<5.918 13.8434 -9.68>, rot=<-0.2163 0.4376 0.8119 0.3203>}
 AD_Med_pine_rowA_A1 {parent=spine_04, loc=<8.1629 13.6948 -3.2552>, rot=<-0.1415 0.3903 0.8781 0.2379>}
 AD_Med_pine_rowB_A1 {parent=spine_04, loc=<5.39 10.7163 13.5404>, rot=<0.2518 -0.6517 0.6865 0.2015>}
 AD_Sm_pine_rowB_B1 {parent=spine_04, loc=<1.771 11.547 11.1456>, rot=<0.2655 -0.6597 0.6718 0.2074>}
 AD_Med_pine_rowB_F1 {parent=spine_04, loc=<0.7854 12.0273 -11.6911>, rot=<-0.1309 0.5907 0.7889 0.1071>}
 AD_Sm_pine_rowB_E1 {parent=spine_04, loc=<-0.1069 14.0827 -4.4825>, rot=<-0.258 0.5166 0.7582 0.3028>}
 AD_Lg_pine_rowB_D1 {parent=spine_04, loc=<-0.6992 13.7907 4.1057>, rot=<0.0599 -0.2113 0.9574 0.1872>}
 AD_Sm_pine_rowA_Q1 {parent=spine_04, loc=<7.3418 12.644 7.1548>, rot=<0.2143 -0.5226 0.7869 0.2485>}
 CR_OrnamentA1 {parent=spine_04, loc=<1.4213 10.2598 15.228>, rot=<0.6665 0.2362 0.6526 0.2722>}
 CR_SashOrnamentA1 {parent=spine_04, loc=<0.8747 14.6444 2.9072>, rot=<-0.295 -0.2459 -0.6648 -0.6408>}
 AD_pine_rowB_L1 {parent=spine_03, loc=<10.4378 -9.2245 -0.8052>, rot=<0.3834 -0.9142 0.1261 0.0373>}
 AD_pine_rowB_M1 {parent=spine_03, loc=<11.0766 -10.0978 1.4098>, rot=<0.3461 -0.9183 0.1863 0.0477>}
 AD_pine_rowB_P1 {parent=spine_03, loc=<9.439 -10.7673 4.6308>, rot=<0.2955 -0.7661 0.5518 0.146>}
 AD_pine_rowB_S1 {parent=spine_03, loc=<12.2217 -9.1558 9.2018>, rot=<0.1733 -0.6781 0.7072 0.1002>}
 AD_pine_rowB_T1 {parent=spine_03, loc=<14.0339 -2.4813 12.6422>, rot=<0.2581 -0.8279 0.4888 0.0953>}
 AD_pine_rowC_E1 {parent=spine_03, loc=<8.5786 -6.092 -11.0879>, rot=<0.349 -0.824 -0.4276 -0.1281>}
 AD_pine_rowB_H1 {parent=spine_03, loc=<11.5406 -0.0193 -16.3576>, rot=<0.2095 -0.7891 -0.571 -0.0863>}
 AD_pine_rowB_U1 {parent=spine_03, loc=<12.1791 0.8766 16.3526>, rot=<0.1098 -0.8529 0.5104 0.0094>}
 AD_pine_rowB_W1 {parent=spine_03, loc=<10.1078 11.3713 10.8669>, rot=<0.1798 -0.7673 0.6111 0.0741>}
 AD_pine_rowC_A1 {parent=spine_03, loc=<10.1437 12.7556 -5.3138>, rot=<-0.1854 0.5914 0.77 0.1515>}
 AD_pine_rowC_C1 {parent=spine_03, loc=<10.0397 7.0267 -15.7247>, rot=<0.1902 -0.8699 -0.4526 -0.048>}
 AD_pine_rowA_U1 {parent=spine_03, loc=<10.5695 12.1682 1.1905>, rot=<-0.0824 0.1943 0.9363 0.2808>}
 AD_Sm_pine_rowB_L1 {parent=spine_03, loc=<4.9803 -14.5924 0.4744>, rot=<0.3834 -0.9142 0.1261 0.0373>}
 AD_Sm_pine_rowB_M1 {parent=spine_03, loc=<5.2101 -14.8931 3.0909>, rot=<0.3461 -0.9183 0.1863 0.0477>}
 AD_Sm_pine_rowB_P1 {parent=spine_03, loc=<3.3643 -13.0308 8.8984>, rot=<0.2955 -0.7661 0.5518 0.146>}
 AD_Sm_pine_rowB_S1 {parent=spine_03, loc=<5.0823 -9.8801 12.1583>, rot=<0.1733 -0.6781 0.7072 0.1002>}
 AD_Sm_pine_rowB_T1 {parent=spine_03, loc=<7.4476 -5.0744 15.8256>, rot=<0.2581 -0.8279 0.4888 0.0953>}
 AD_Med_pine_rowC_E1 {parent=spine_03, loc=<2.9634 -9.7059 -15.0435>, rot=<0.349 -0.824 -0.4276 -0.1281>}
 AD_Sm_pine_rowB_H1 {parent=spine_03, loc=<4.5762 -1.8205 -19.2713>, rot=<0.2095 -0.7891 -0.571 -0.0863>}
 AD_Sm_pine_rowB_U1 {parent=spine_03, loc=<4.6062 -0.5019 17.3471>, rot=<0.1098 -0.8529 0.5104 0.0094>}
 AD_Sm_pine_rowB_W1 {parent=spine_03, loc=<2.9338 9.9324 13.4552>, rot=<0.1798 -0.7673 0.6111 0.0741>}
 AD_Med_pine_rowC_A1 {parent=spine_03, loc=<3.2721 12.8647 -8.9203>, rot=<-0.1854 0.5914 0.77 0.1515>}
 AD_Med_pine_rowC_C1 {parent=spine_03, loc=<2.8758 4.796 -17.7096>, rot=<0.1902 -0.8699 -0.4526 -0.048>}
 AD_Sm_pine_rowA_U1 {parent=spine_03, loc=<4.1379 16.0013 -0.8537>, rot=<-0.0824 0.1943 0.9363 0.2808>}
 CR_SashOrnamentB1 {parent=spine_03, loc=<3.1042 13.1634 8.2378>, rot=<-0.3445 -0.2029 -0.7256 -0.56>}
 AD_pine_rowB_O1 {parent=spine_02, loc=<11.1518 -11.7492 1.4745>, rot=<0.3851 -0.8139 0.4068 0.1541>}
 AD_pine_rowC_D1 {parent=spine_02, loc=<9.6775 -1.2151 -12.9661>, rot=<0.3223 -0.6972 -0.6008 -0.2215>}
 AD_pine_rowC_F1 {parent=spine_02, loc=<7.9 -9.8317 -5.9127>, rot=<0.4295 -0.8302 -0.3276 -0.1376>}
 AD_pine_rowC_G1 {parent=spine_02, loc=<10.0844 -12.8559 2.5569>, rot=<0.3783 -0.8303 -0.3848 -0.1395>}
 AD_pine_rowC_H1 {parent=spine_02, loc=<6.362 -13.773 -1.1451>, rot=<0.4063 -0.8653 0.2747 0.1032>}
 AD_pine_rowC_I1 {parent=spine_02, loc=<5.1041 -13.1949 2.8146>, rot=<0.4541 -0.7848 0.3807 0.1815>}
 AD_pine_rowC_J1 {parent=spine_02, loc=<7.4727 -11.1735 6.2429>, rot=<0.319 -0.7337 0.5675 0.1946>}
 AD_pine_rowC_M1 {parent=spine_02, loc=<5.8826 -7.2481 11.2223>, rot=<0.2415 -0.6257 0.7111 0.2111>}
 AD_pine_rowC_N1 {parent=spine_02, loc=<5.0154 -1.8961 11.7364>, rot=<0.2668 -0.5331 0.7443 0.301>}
 AD_pine_rowB_V1 {parent=spine_02, loc=<12.5465 3.5228 17.4453>, rot=<0.0007 -0.0056 0.9993 0.0357>}
 AD_pine_rowC_B1 {parent=spine_02, loc=<12.0711 8.4696 -9.4754>, rot=<-0.1489 0.5151 0.827 0.1691>}
 AD_pine_rowC_T1 {parent=spine_02, loc=<7.8398 9.8643 -1.079>, rot=<-0.1687 0.3692 0.8588 0.3124>}
 AD_pine_rowC_U1 {parent=spine_02, loc=<3.9637 4.8287 -10.5258>, rot=<-0.3155 0.5444 0.7014 0.3349>}
 AD_pine_rowC_R1 {parent=spine_02, loc=<6.9192 11.1586 3.4452>, rot=<-0.0577 0.1569 0.9495 0.2656>}
 AD_pine_rowC_O1 {parent=spine_02, loc=<5.6439 5.6574 10.7998>, rot=<0.2017 -0.4782 0.8115 0.2685>}
 AD_pine_rowC_S1 {parent=spine_02, loc=<11.0119 11.4977 0.802>, rot=<-0.1138 0.3578 0.9028 0.2097>}
 AD_pine_rowA_V1 {parent=spine_02, loc=<9.4342 12.5329 2.1773>, rot=<0.2793 -0.5896 0.7086 0.2688>}
 AD_Sm_pine_rowB_O1 {parent=spine_02, loc=<6.0612 -15.6408 5.8539>, rot=<0.3851 -0.8139 0.4068 0.1541>}
 AD_Lg_pine_rowC_D1 {parent=spine_02, loc=<4.2897 -2.6368 -18.3687>, rot=<0.3223 -0.6972 -0.6008 -0.2215>}
 AD_Lg_pine_rowC_F1 {parent=spine_02, loc=<3.2966 -14.6673 -9.8704>, rot=<0.4295 -0.8302 -0.3276 -0.1376>}
 AD_Med_pine_rowC_G1 {parent=spine_02, loc=<4.8465 -16.8981 -1.5005>, rot=<0.3783 -0.8303 -0.3848 -0.1395>}
 AD_Med_pine_rowC_H1 {parent=spine_02, loc=<1.3287 -18.7906 1.9734>, rot=<0.4063 -0.8653 0.2747 0.1032>}
 AD_Sm_pine_rowC_I1 {parent=spine_02, loc=<1.0556 -17.6547 7.7091>, rot=<0.4541 -0.7848 0.3807 0.1815>}
 AD_Med_pine_rowC_J1 {parent=spine_02, loc=<1.8785 -13.0921 11.2689>, rot=<0.319 -0.7337 0.5675 0.1946>}
 AD_Lg_pine_rowC_M1 {parent=spine_02, loc=<-0.281 -7.2635 15.9389>, rot=<0.2415 -0.6257 0.7111 0.2111>}
 AD_Lg_pine_rowC_N1 {parent=spine_02, loc=<-0.2342 -0.627 17.3103>, rot=<0.2668 -0.5331 0.7443 0.301>}
 AD_Sm_pine_rowB_V1 {parent=spine_02, loc=<4.8051 4.0763 17.4585>, rot=<0.0007 -0.0056 0.9993 0.0357>}
 AD_Sm_pine_rowC_B1 {parent=spine_02, loc=<5.0979 9.4493 -12.739>, rot=<-0.1489 0.5151 0.827 0.1691>}
 AD_Sm_pine_rowC_T1 {parent=spine_02, loc=<2.0356 13.0623 -5.1189>, rot=<-0.1687 0.3692 0.8588 0.3124>}
 AD_Med_pine_rowC_U1 {parent=spine_02, loc=<-0.5116 5.808 -16.7908>, rot=<-0.3155 0.5444 0.7014 0.3349>}
 AD_Med_pine_rowC_R1 {parent=spine_02, loc=<0.3048 14.9329 1.9479>, rot=<-0.0577 0.1569 0.9495 0.2656>}
 AD_Med_pine_rowC_O1 {parent=spine_02, loc=<-0.3669 7.5416 15.3338>, rot=<0.2017 -0.4782 0.8115 0.2685>}
 AD_Sm_pine_rowC_S1 {parent=spine_02, loc=<4.1345 13.8047 -1.9578>, rot=<-0.1138 0.3578 0.9028 0.2097>}
 AD_Sm_pine_rowA_V1 {parent=spine_02, loc=<4.0055 12.9345 7.7094>, rot=<0.2793 -0.5896 0.7086 0.2688>}
 AD_pine_rowC_L1 {parent=spine_01, loc=<5.8072 -11.93 10.0168>, rot=<0.3264 -0.6887 0.5937 0.2583>}
 AD_pine_rowC_K1 {parent=spine_01, loc=<8.605 -14.6529 2.734>, rot=<0.412 -0.7107 0.5023 0.2699>}
 AD_pine_rowD_M1 {parent=spine_01, loc=<5.5825 -14.4999 -3.2086>, rot=<0.4278 -0.8136 0.3542 0.1719>}
 AD_pine_rowD_N1 {parent=spine_01, loc=<8.3424 -11.9704 4.9385>, rot=<0.3571 -0.899 0.2386 0.086>}
 AD_pine_rowD_L1 {parent=spine_01, loc=<3.5497 -13.4836 -8.6832>, rot=<0.4206 -0.8176 -0.3552 -0.1685>}
 AD_pine_rowD_K1 {parent=spine_01, loc=<12.3797 -6.1479 -12.8928>, rot=<0.2811 -0.8631 -0.4029 -0.117>}
 AD_pine_rowC_Q1 {parent=spine_01, loc=<9.289 10.643 3.2357>, rot=<0.1522 -0.3015 0.8535 0.3968>}
 AD_pine_rowC_P1 {parent=spine_01, loc=<9.3134 11.1061 6.2198>, rot=<0.2786 -0.5121 0.7258 0.3651>}
 AD_pine_rowD_A1 {parent=spine_01, loc=<6.14 3.7995 14.5378>, rot=<0.1725 -0.5919 0.7626 0.1958>}
 AD_pine_rowD_B1 {parent=spine_01, loc=<5.6765 10.7179 10.2713>, rot=<0.2089 -0.5869 0.7451 0.2384>}
 AD_pine_rowD_F1 {parent=spine_01, loc=<7.1045 9.3923 2.7224>, rot=<-0.109 0.2616 0.8967 0.34>}
 AD_pine_rowD_H1 {parent=spine_01, loc=<8.6654 10.8738 -5.4308>, rot=<-0.2218 0.5255 0.7666 0.2948>}
 AD_pine_rowD_G1 {parent=spine_01, loc=<11.1169 10.4077 1.9053>, rot=<-0.2064 0.473 0.7957 0.317>}
 AD_pine_rowD_J1 {parent=spine_01, loc=<11.7582 -0.0593 -13.7588>, rot=<0.2398 -0.7072 -0.6364 -0.1932>}
 AD_pine_rowD_I1 {parent=spine_01, loc=<10.3563 7.7795 -8.6518>, rot=<-0.2138 0.5395 0.7664 0.2754>}
 AD_pine_rowD_C1 {parent=spine_01, loc=<5.7434 10.8227 10.0228>, rot=<0.0388 -0.1428 0.9622 0.2285>}
 AD_Med_pine_rowC_L1 {parent=spine_01, loc=<0.7357 -13.0392 15.7862>, rot=<0.3264 -0.6887 0.5937 0.2583>}
 AD_Med_pine_rowC_K1 {parent=spine_01, loc=<4.6101 -17.094 8.9241>, rot=<0.412 -0.7107 0.5023 0.2699>}
 AD_Med_pine_rowD_M1 {parent=spine_01, loc=<1.1208 -18.958 1.3142>, rot=<0.4278 -0.8136 0.3542 0.1719>}
 AD_Med_pine_rowD_N1 {parent=spine_01, loc=<2.6756 -16.6354 7.4608>, rot=<0.3571 -0.899 0.2386 0.086>}
 AD_Med_pine_rowD_L1 {parent=spine_01, loc=<-1.0249 -17.8933 -13.1402>, rot=<0.4206 -0.8176 -0.3552 -0.1685>}
 AD_Med_pine_rowD_K1 {parent=spine_01, loc=<6.0578 -9.1827 -16.2184>, rot=<0.2811 -0.8631 -0.4029 -0.117>}
 AD_Med_pine_rowC_Q1 {parent=spine_01, loc=<4.3313 15.1881 7.1086>, rot=<0.1522 -0.3015 0.8535 0.3968>}
 AD_Sm_pine_rowC_P1 {parent=spine_01, loc=<4.8261 13.0042 12.2612>, rot=<0.2786 -0.5121 0.7258 0.3651>}
 AD_Lg_pine_rowD_A1 {parent=spine_01, loc=<-0.5644 4.532 18.3786>, rot=<0.1725 -0.5919 0.7626 0.1958>}
 AD_Lg_pine_rowD_B1 {parent=spine_01, loc=<-0.5248 11.5725 14.8594>, rot=<0.2089 -0.5869 0.7451 0.2384>}
 AD_Med_pine_rowD_F1 {parent=spine_01, loc=<1.322 13.6825 -0.1749>, rot=<-0.109 0.2616 0.8967 0.34>}
 AD_Med_pine_rowD_H1 {parent=spine_01, loc=<3.0173 12.5725 -10.4756>, rot=<-0.2218 0.5255 0.7666 0.2948>}
 AD_Med_pine_rowD_G1 {parent=spine_01, loc=<5.5769 12.8085 -2.9714>, rot=<-0.2064 0.473 0.7957 0.317>}
 AD_Med_pine_rowD_J1 {parent=spine_01, loc=<5.469 -0.784 -18.2484>, rot=<0.2398 -0.7072 -0.6364 -0.1932>}
 AD_Med_pine_rowD_I1 {parent=spine_01, loc=<4.4821 9.2659 -13.5017>, rot=<-0.2138 0.5395 0.7664 0.2754>}
 AD_Sm_pine_rowD_C1 {parent=spine_01, loc=<-1.1841 14.1495 11.109>, rot=<0.0388 -0.1428 0.9622 0.2285>}
 CR_CandyCane1 {parent=spine_01, loc=<4.1067 13.7521 -3.6506>, rot=<0.0114 -0.707 -0.0114 -0.707>}
 thigh_l {parent=pelvis, loc=<-2.5693 -0.0109 -7.7501>, rot=<0.0012 0.0572 0.0155 -0.9982>}
 calf_l {parent=thigh_l, loc=<-39.0484 0 0>, rot=<0.0006 -0.0161 -0.0391 -0.9991>}
 foot_l {parent=calf_l, loc=<-42.561 0.1077 0.4292>, rot=<0.0418 -0.0455 0.0343 -0.9975>}
 ball_l {parent=foot_l, loc=<-6.1599 13.8264 0.6806>, rot=<0.0001 -0.0001 -0.7063 -0.7079>}
 calf_twist_02_l {parent=calf_l, loc=<-18.8064 -0.0967 0.0397>, rot=<0.0237 0.0095 -0.0035 -0.9997>}
 calf_twist_01_l {parent=calf_l, loc=<-35.813 -0.0967 0.0397>, rot=<0.0237 0.0095 -0.0035 -0.9997>}
 deform_knee_in_base_l {parent=calf_l, loc=<1.298 -5.2957 -0.0496>, rot=<0 0 0 -1>}
 deform_knee_in_l {parent=deform_knee_in_base_l, loc=<0 0 0>, rot=<0 0 0 1>}
 deform_knee_out_base_l {parent=calf_l, loc=<0 4 0>, rot=<0 0 0 -1>}
 deform_knee_out_l {parent=deform_knee_out_base_l, loc=<0 0 0>, rot=<0 0 0 1>}
 thigh_twist_01_l {parent=thigh_l, loc=<-21.6303 0 0>, rot=<0.0013 0 0 -1>}
 thigh_r {parent=pelvis, loc=<-2.5693 -0.0109 7.7501>, rot=<0.0572 -0.0012 0.9982 0.0155>}
 calf_r {parent=thigh_r, loc=<39.0484 0 0>, rot=<0.0006 -0.0161 -0.0391 -0.9991>}
 foot_r {parent=calf_r, loc=<42.561 -0.1077 -0.4292>, rot=<0.0418 -0.0455 0.0343 -0.9975>}
 ball_r {parent=foot_r, loc=<6.1599 -13.8264 -0.6806>, rot=<0.0001 -0.0001 -0.7063 -0.7079>}
 calf_twist_02_r {parent=calf_r, loc=<18.8064 0.0967 -0.0397>, rot=<0.0237 0.0095 -0.0035 -0.9997>}
 calf_twist_01_r {parent=calf_r, loc=<35.813 0.0967 -0.0397>, rot=<0.0237 0.0095 -0.0035 -0.9997>}
 deform_knee_in_base_r {parent=calf_r, loc=<-1.298 5.2957 0.0496>, rot=<0 0 0 -1>}
 deform_knee_in_r {parent=deform_knee_in_base_r, loc=<0 0 0>, rot=<0 0 0 1>}
 deform_knee_out_base_r {parent=calf_r, loc=<0 -4 0>, rot=<0 0 0 -1>}
 deform_knee_out_r {parent=deform_knee_out_base_r, loc=<0 0 0>, rot=<0 0 0 1>}
 thigh_twist_01_r {parent=thigh_r, loc=<21.6303 0 0>, rot=<0.0013 0 0 -1>}
 MSTR_skirt_fr {parent=pelvis, loc=<0.4429 15.0864 0>, rot=<0 0 0.997 0.0772>}
 dyn_skirt_fr_01 {parent=MSTR_skirt_fr, loc=<0 0 0>, rot=<0 0 0 1>}
 dyn_skirt_fr_02 {parent=dyn_skirt_fr_01, loc=<8.3253 0 0>, rot=<0 0 0.0061 -1>}
 dyn_skirt_fr_03 {parent=dyn_skirt_fr_02, loc=<7.9753 0 0>, rot=<0 0 0 1>}
 dyn_skirt_fr_04 {parent=dyn_skirt_fr_03, loc=<9.0514 0 0>, rot=<0 0 0 1>}
 AD_pine_rowF_D1 {parent=dyn_skirt_fr_01, loc=<3.8234 2.7674 -5.3791>, rot=<-0.2952 -0.1414 0.3043 -0.8946>}
 AD_pine_rowF_D2 {parent=AD_pine_rowF_D1, loc=<2.5246 0 0>, rot=<0 0 0 1>}
 AD_pine_rowF_B1 {parent=dyn_skirt_fr_01, loc=<2.798 1.65 4.906>, rot=<-0.2485 -0.0857 0.2045 -0.9429>}
 AD_pine_rowF_B2 {parent=AD_pine_rowF_B1, loc=<2.9355 0 0>, rot=<0 0 0 1>}
 AD_pine_rowF_A1 {parent=dyn_skirt_fr_01, loc=<4.1518 1.0325 8.1656>, rot=<-0.1073 -0.0379 0.2179 -0.9693>}
 AD_pine_rowF_A2 {parent=AD_pine_rowF_A1, loc=<2.2987 0 0>, rot=<0 0 0 1>}
 AD_pine_rowF_C1 {parent=dyn_skirt_fr_01, loc=<6.2821 0.8181 1.1286>, rot=<-0.3402 -0.1703 0.3131 -0.8702>}
 AD_pine_rowF_C2 {parent=AD_pine_rowF_C1, loc=<-0.1597 0 0>, rot=<0 0 0 1>}
 AD_pine_rowG_J1 {parent=dyn_skirt_fr_01, loc=<12.4782 1.2689 -5.3716>, rot=<-0.3989 -0.1811 0.2721 -0.8568>}
 AD_pine_rowG_J2 {parent=AD_pine_rowG_J1, loc=<5.3166 0 0>, rot=<0 0 0 1>}
 AD_pine_rowG_I1 {parent=dyn_skirt_fr_01, loc=<15.943 1.8491 5.3432>, rot=<-0.2677 -0.131 0.315 -0.9011>}
 AD_pine_rowG_I2 {parent=AD_pine_rowG_I1, loc=<2.6899 0 0>, rot=<0 0 0 1>}
 AD_pine_rowE_E1 {parent=dyn_skirt_fr_01, loc=<-5.1562 1.7471 4.6799>, rot=<0.1978 0.0655 0.1954 -0.9583>}
 AD_pine_rowE_E2 {parent=AD_pine_rowE_E1, loc=<2.8326 0 0>, rot=<0 0 0 1>}
 AD_pine_rowE_F1 {parent=dyn_skirt_fr_01, loc=<-5.0838 1.2363 6.1806>, rot=<-0.3318 -0.1455 0.2706 -0.8919>}
 AD_pine_rowE_F2 {parent=AD_pine_rowE_F1, loc=<3.5611 0 0>, rot=<0 0 0 1>}
 AD_pine_rowE_G1 {parent=dyn_skirt_fr_01, loc=<-2.5096 3.6977 0.0076>, rot=<-0.2113 -0.1106 0.3453 -0.9077>}
 AD_pine_rowE_G2 {parent=AD_pine_rowE_G1, loc=<3.8857 0 0>, rot=<0 0 0 1>}
 AD_Lg_pine_rowF_D1 {parent=dyn_skirt_fr_01, loc=<9.8367 -0.8105 -8.737>, rot=<-0.2952 -0.1414 0.3043 -0.8946>}
 AD_Sm_pine_rowF_B1 {parent=dyn_skirt_fr_01, loc=<9.7958 -1.0127 2.8624>, rot=<-0.2485 -0.0857 0.2045 -0.9429>}
 AD_Lg_pine_rowF_A1 {parent=dyn_skirt_fr_01, loc=<11.154 -2.1822 7.2318>, rot=<-0.1073 -0.0379 0.2179 -0.9693>}
 AD_Lg_pine_rowF_C1 {parent=dyn_skirt_fr_01, loc=<12.0715 -2.5115 -2.8253>, rot=<-0.3402 -0.1703 0.3131 -0.8702>}
 AD_Lg_pine_rowG_J1 {parent=dyn_skirt_fr_01, loc=<18.5814 -1.2278 -9.4648>, rot=<-0.3989 -0.1811 0.2721 -0.8568>}
 AD_Lg_pine_rowG_I1 {parent=dyn_skirt_fr_01, loc=<21.8973 -2.0133 2.2024>, rot=<-0.2677 -0.131 0.315 -0.9011>}
 AD_Sm_pine_rowE_E1 {parent=dyn_skirt_fr_01, loc=<1.946 -0.9581 6.2541>, rot=<0.1978 0.0655 0.1954 -0.9583>}
 AD_Sm_pine_rowE_F1 {parent=dyn_skirt_fr_01, loc=<1.2126 -1.7603 2.7729>, rot=<-0.3318 -0.1455 0.2706 -0.8919>}
 AD_Sm_pine_rowE_G1 {parent=dyn_skirt_fr_01, loc=<3.2113 -0.8042 -2.6835>, rot=<-0.2113 -0.1106 0.3453 -0.9077>}
 MSTR_skirtA_l {parent=pelvis, loc=<0.2547 9.8358 -13.9059>, rot=<-0.0547 0.3847 0.9182 0.0765>}
 dyn_skirtA_l_01 {parent=MSTR_skirtA_l, loc=<0 0 0>, rot=<0 0 0 1>}
 dyn_skirtA_l_02 {parent=dyn_skirtA_l_01, loc=<8.3909 0 0>, rot=<-0.0006 -0.0034 0.008 -1>}
 dyn_skirtA_l_03 {parent=dyn_skirtA_l_02, loc=<8.0516 0 0>, rot=<0 0 0 1>}
 dyn_skirtA_l_04 {parent=dyn_skirtA_l_03, loc=<9.138 0 0>, rot=<0 0 0 1>}
 AD_pine_rowE_I1 {parent=dyn_skirtA_l_01, loc=<-4.869 4.5586 7.3822>, rot=<-0.1482 -0.0493 0.3588 -0.9202>}
 AD_pine_rowE_I2 {parent=AD_pine_rowE_I1, loc=<4.8721 0 0>, rot=<0 0 0 1>}
 AD_pine_rowF_F1 {parent=dyn_skirtA_l_01, loc=<-2.5197 2.6258 -2.6914>, rot=<-0.312 -0.1274 0.2882 -0.8963>}
 AD_pine_rowF_F2 {parent=AD_pine_rowF_F1, loc=<5.4261 0 0>, rot=<0 0 0 1>}
 AD_pine_rowF_E1 {parent=dyn_skirtA_l_01, loc=<3.357 0.7018 5.1892>, rot=<-0.24 -0.075 0.2539 -0.934>}
 AD_pine_rowF_E2 {parent=AD_pine_rowF_E1, loc=<3.4961 0 0>, rot=<0 0 0 1>}
 AD_pine_rowG_K1 {parent=dyn_skirtA_l_01, loc=<13.7333 -1.0347 1.5939>, rot=<-0.303 -0.1035 0.2425 -0.9158>}
 AD_pine_rowG_K2 {parent=AD_pine_rowG_K1, loc=<3.9795 0 0>, rot=<0 0 0 1>}
 AD_pine_rowG_L1 {parent=dyn_skirtA_l_01, loc=<16.2167 2.2422 -5.6104>, rot=<-0.2456 -0.089 0.2894 -0.9209>}
 AD_pine_rowG_L2 {parent=AD_pine_rowG_L1, loc=<0.1917 0 0>, rot=<0 0 0 1>}
 AD_Sm_pine_rowE_I1 {parent=dyn_skirtA_l_01, loc=<0.8558 -0.4536 5.852>, rot=<-0.1482 -0.0493 0.3588 -0.9202>}
 AD_Lg_pine_rowF_F1 {parent=dyn_skirtA_l_01, loc=<3.7005 -0.7665 -5.8598>, rot=<-0.312 -0.1274 0.2882 -0.8963>}
 AD_Lg_pine_rowF_E1 {parent=dyn_skirtA_l_01, loc=<10.03 -2.7 3.1557>, rot=<-0.24 -0.075 0.2539 -0.934>}
 AD_Lg_pine_rowG_K1 {parent=dyn_skirtA_l_01, loc=<20.4155 -3.9947 -1.0184>, rot=<-0.303 -0.1035 0.2425 -0.9158>}
 AD_Sm_pine_rowG_L1 {parent=dyn_skirtA_l_01, loc=<22.5553 -1.5548 -7.9857>, rot=<-0.2456 -0.089 0.2894 -0.9209>}
 MSTR_skirtB_l {parent=pelvis, loc=<-0.1997 -2.8402 -19.6659>, rot=<-0.0908 0.7013 0.7039 0.0676>}
 dyn_skirtB_l_01 {parent=MSTR_skirtB_l, loc=<0 0 0>, rot=<0 0 0 1>}
 dyn_skirtB_l_02 {parent=dyn_skirtB_l_01, loc=<8.4994 0 0>, rot=<0.0008 -0.0035 0.0111 -0.9999>}
 dyn_skirtB_l_03 {parent=dyn_skirtB_l_02, loc=<8.1775 0 0>, rot=<0 0 0 1>}
 dyn_skirtB_l_04 {parent=dyn_skirtB_l_03, loc=<9.2808 0 0>, rot=<0 0 0 1>}
 AD_pine_rowG_M1 {parent=dyn_skirtB_l_01, loc=<7.217 1.4597 4.9179>, rot=<-0.0027 0.0371 0.2376 -0.9707>}
 AD_pine_rowG_M2 {parent=AD_pine_rowG_M1, loc=<9.9218 0 0>, rot=<0 0 0 1>}
 AD_pine_rowF_G1 {parent=dyn_skirtB_l_01, loc=<-1.8108 2.8133 1.9465>, rot=<-0.1377 -0.0439 0.3087 -0.9401>}
 AD_pine_rowF_G2 {parent=AD_pine_rowF_G1, loc=<4.948 0 0>, rot=<0 0 0 1>}
 AD_pine_rowG_N1 {parent=dyn_skirtB_l_01, loc=<17.7765 3.3324 -1.2024>, rot=<-0.0399 0.0141 0.3754 -0.9259>}
 AD_pine_rowG_N2 {parent=AD_pine_rowG_N1, loc=<0.1677 0 0>, rot=<0 0 0 1>}
 AD_pine_rowF_H1 {parent=dyn_skirtB_l_01, loc=<10.5391 0.4238 -5.546>, rot=<-0.088 -0.0051 0.2274 -0.9698>}
 AD_pine_rowF_H2 {parent=AD_pine_rowF_H1, loc=<2.3936 0 0>, rot=<0 0 0 1>}
 AD_Lg_pine_rowG_M1 {parent=dyn_skirtB_l_01, loc=<14.0807 -2.1215 5.4676>, rot=<-0.0027 0.0371 0.2376 -0.9707>}
 AD_Lg_pine_rowF_G1 {parent=dyn_skirtB_l_01, loc=<4.4416 -1.5975 0.6469>, rot=<-0.1377 -0.0439 0.3087 -0.9401>}
 AD_Lg_pine_rowG_N1 {parent=dyn_skirtB_l_01, loc=<23.3467 -2.0721 -1.2329>, rot=<-0.0399 0.0141 0.3754 -0.9259>}
 AD_Lg_pine_rowF_H1 {parent=dyn_skirtB_l_01, loc=<17.4975 -2.992 -5.9333>, rot=<-0.088 -0.0051 0.2274 -0.9698>}
 MSTR_skirtC_l {parent=pelvis, loc=<-0.654 -15.5163 -13.9059>, rot=<0.1052 -0.9155 -0.3862 -0.0412>}
 dyn_skirtC_l_01 {parent=MSTR_skirtC_l, loc=<0 0 0>, rot=<0 0 0 1>}
 dyn_skirtC_l_02 {parent=dyn_skirtC_l_01, loc=<8.5391 0 0>, rot=<0.0016 -0.0016 0.0124 -0.9999>}
 dyn_skirtC_l_03 {parent=dyn_skirtC_l_02, loc=<8.2234 0 0>, rot=<0 0 0 1>}
 dyn_skirtC_l_04 {parent=dyn_skirtC_l_03, loc=<9.3329 0 0>, rot=<0 0 0 1>}
 AD_pine_rowE_L1 {parent=dyn_skirtC_l_01, loc=<-7.7162 -0.2614 2.0251>, rot=<0.3259 0.1704 0.1818 -0.912>}
 AD_pine_rowE_L2 {parent=AD_pine_rowE_L1, loc=<5.2493 0 0>, rot=<0 0 0 1>}
 AD_pine_rowF_J1 {parent=dyn_skirtC_l_01, loc=<4.5616 4.6156 -1.429>, rot=<0.0718 0.0584 0.2584 -0.9616>}
 AD_pine_rowF_J2 {parent=AD_pine_rowF_J1, loc=<5.4779 0 0>, rot=<0 0 0 1>}
 AD_pine_rowF_I1 {parent=dyn_skirtC_l_01, loc=<12.9643 1.6009 5.2576>, rot=<0.0898 0.0727 0.2887 -0.9504>}
 AD_pine_rowF_I2 {parent=AD_pine_rowF_I1, loc=<3.4683 0 0>, rot=<0 0 0 1>}
 AD_pine_rowF_L1 {parent=dyn_skirtC_l_01, loc=<18.9605 1.5637 -4.2292>, rot=<-0.0262 -0.0006 0.4042 -0.9143>}
 AD_pine_rowF_L2 {parent=AD_pine_rowF_L1, loc=<0.3237 0 0>, rot=<0 0 0 1>}
 AD_pine_rowF_N1 {parent=dyn_skirtC_l_01, loc=<14.7969 2.3838 -8.348>, rot=<-0.04 -0.0083 0.3392 -0.9398>}
 AD_pine_rowF_N2 {parent=AD_pine_rowF_N1, loc=<-0.5934 0 0>, rot=<0 0 0 1>}
 AD_pine_rowF_K1 {parent=dyn_skirtC_l_01, loc=<20.8841 2.2486 0.3893>, rot=<0.0447 0.0663 0.504 -0.86>}
 AD_pine_rowF_K2 {parent=AD_pine_rowF_K1, loc=<-1.4929 0 0>, rot=<0 0 0 1>}
 AD_Sm_pine_rowE_L1 {parent=dyn_skirtC_l_01, loc=<-0.9184 -1.9725 5.3566>, rot=<0.3259 0.1704 0.1818 -0.912>}
 AD_Lg_pine_rowF_J1 {parent=dyn_skirtC_l_01, loc=<11.2337 0.824 -0.2695>, rot=<0.0718 0.0584 0.2584 -0.9616>}
 AD_Sm_pine_rowF_I1 {parent=dyn_skirtC_l_01, loc=<19.3495 -2.5574 6.7322>, rot=<0.0898 0.0727 0.2887 -0.9504>}
 AD_Lg_pine_rowF_L1 {parent=dyn_skirtC_l_01, loc=<24.1862 -4.1721 -4.4021>, rot=<-0.0262 -0.0006 0.4042 -0.9143>}
 AD_Lg_pine_rowF_N1 {parent=dyn_skirtC_l_01, loc=<20.771 -2.5596 -8.6793>, rot=<-0.04 -0.0083 0.3392 -0.9398>}
 AD_Sm_pine_rowF_K1 {parent=dyn_skirtC_l_01, loc=<24.634 -4.4335 1.6238>, rot=<0.0447 0.0663 0.504 -0.86>}
 MSTR_skirt_ba {parent=pelvis, loc=<-0.8422 -20.7669 0>, rot=<0.1077 -0.9942 0 0>}
 dyn_skirt_ba_01 {parent=MSTR_skirt_ba, loc=<0 0 0>, rot=<0 0 0 1>}
 dyn_skirt_ba_02 {parent=dyn_skirt_ba_01, loc=<8.5356 0 0>, rot=<0 0 0.0124 -0.9999>}
 dyn_skirt_ba_03 {parent=dyn_skirt_ba_02, loc=<8.2194 0 0>, rot=<0 0 0 1>}
 dyn_skirt_ba_04 {parent=dyn_skirt_ba_03, loc=<9.3284 0 0>, rot=<0 0 0 1>}
 AD_pine_rowA_S1 {parent=dyn_skirt_ba_01, loc=<-9.5038 3.7116 -0.2218>, rot=<0.2102 0.1262 0.2763 -0.9293>}
 AD_pine_rowA_S2 {parent=AD_pine_rowA_S1, loc=<-0.6267 0 0>, rot=<0 0 0 1>}
 AD_pine_rowA_R1 {parent=dyn_skirt_ba_01, loc=<-9.0597 2.2346 1.7938>, rot=<0.3305 0.2156 0.2947 -0.8703>}
 AD_pine_rowA_R2 {parent=AD_pine_rowA_R1, loc=<-1.267 0 0>, rot=<0 0 0 1>}
 AD_pine_rowE_M1 {parent=dyn_skirt_ba_01, loc=<1.4039 3.3654 -0.4284>, rot=<0.2832 0.2682 0.4467 -0.8052>}
 AD_pine_rowE_M2 {parent=AD_pine_rowE_M1, loc=<0.7618 0 0>, rot=<0 0 0 1>}
 AD_pine_rowF_M1 {parent=dyn_skirt_ba_01, loc=<6.5491 2.3345 3.342>, rot=<0.3178 0.2198 0.3192 -0.8653>}
 AD_pine_rowF_M2 {parent=AD_pine_rowF_M1, loc=<0.5521 0 0>, rot=<0 0 0 1>}
 AD_pine_rowF_R1 {parent=dyn_skirt_ba_01, loc=<20.5521 2.3386 6.5087>, rot=<0.2077 0.1602 0.3805 -0.8868>}
 AD_pine_rowF_R2 {parent=AD_pine_rowF_R1, loc=<-1.4773 0 0>, rot=<0 0 0 1>}
 AD_pine_rowF_Q1 {parent=dyn_skirt_ba_01, loc=<14.7557 1.1354 -1.7533>, rot=<0.2972 0.1972 0.3064 -0.8825>}
 AD_pine_rowF_Q2 {parent=AD_pine_rowF_Q1, loc=<0.4422 0 0>, rot=<0 0 0 1>}
 AD_pine_rowF_S1 {parent=dyn_skirt_ba_01, loc=<21.3331 0.2883 1.2693>, rot=<0.3315 0.214 0.2904 -0.8718>}
 AD_pine_rowF_S2 {parent=AD_pine_rowF_S1, loc=<-3.4026 0 0>, rot=<0 0 0 1>}
 AD_pine_rowF_O1 {parent=dyn_skirt_ba_01, loc=<5.8721 2.3796 -0.8245>, rot=<0.1518 0.0954 0.2998 -0.937>}
 AD_pine_rowF_O2 {parent=AD_pine_rowF_O1, loc=<0.0411 0 0>, rot=<0 0 0 1>}
 AD_pine_rowF_P1 {parent=dyn_skirt_ba_01, loc=<8.84 0.4789 -7.313>, rot=<0.3053 0.2044 0.3086 -0.8774>}
 AD_pine_rowF_P2 {parent=AD_pine_rowF_P1, loc=<-0.4887 0 0>, rot=<0 0 0 1>}
 AD_pine_rowF_U1 {parent=dyn_skirt_ba_01, loc=<11.0392 4.3696 -8.7578>, rot=<0.1431 0.0988 0.3402 -0.9241>}
 AD_pine_rowF_U2 {parent=AD_pine_rowF_U1, loc=<-1.1014 0 0>, rot=<0 0 0 1>}
 AD_pine_rowF_T1 {parent=dyn_skirt_ba_01, loc=<17.2886 2.679 -0.0401>, rot=<-0.3176 -0.1943 0.2719 -0.8874>}
 AD_pine_rowF_T2 {parent=AD_pine_rowF_T1, loc=<1.7368 0 0>, rot=<0 0 0 1>}
 AD_pine_rowF_V1 {parent=dyn_skirt_ba_01, loc=<16.3474 3.0208 -3.1825>, rot=<-0.3808 -0.2455 0.2812 -0.8459>}
 AD_pine_rowF_V2 {parent=AD_pine_rowF_V1, loc=<2.3205 0 0>, rot=<0 0 0 1>}
 AD_Sm_pine_rowA_S1 {parent=dyn_skirt_ba_01, loc=<-3.1749 0.1376 2.5001>, rot=<0.2102 0.1262 0.2763 -0.9293>}
 AD_Sm_pine_rowA_R1 {parent=dyn_skirt_ba_01, loc=<-3.3683 -0.6404 6.2186>, rot=<0.3305 0.2156 0.2947 -0.8703>}
 AD_Sm_pine_rowE_M1 {parent=dyn_skirt_ba_01, loc=<4.9511 -1.0391 4.8871>, rot=<0.2832 0.2682 0.4467 -0.8052>}
 AD_Lg_pine_rowF_M1 {parent=dyn_skirt_ba_01, loc=<11.979 -0.8689 7.8686>, rot=<0.3178 0.2198 0.3192 -0.8653>}
 AD_Lg_pine_rowF_R1 {parent=dyn_skirt_ba_01, loc=<25.6682 -2.3823 9.9404>, rot=<0.2077 0.1602 0.3805 -0.8868>}
 AD_Lg_pine_rowF_Q1 {parent=dyn_skirt_ba_01, loc=<20.4558 -2.1522 2.3622>, rot=<0.2972 0.1972 0.3064 -0.8825>}
 AD_Lg_pine_rowF_S1 {parent=dyn_skirt_ba_01, loc=<27.0743 -2.5401 5.6598>, rot=<0.3315 0.214 0.2904 -0.8718>}
 AD_Med_Lg_pine_rowF_O1 {parent=dyn_skirt_ba_01, loc=<12.097 -1.7558 1.27>, rot=<0.1518 0.0954 0.2998 -0.937>}
 AD_Lg_pine_rowF_P1 {parent=dyn_skirt_ba_01, loc=<14.4749 -2.755 -3.0672>, rot=<0.3053 0.2044 0.3086 -0.8774>}
 AD_Sm_pine_rowF_U1 {parent=dyn_skirt_ba_01, loc=<16.8519 -0.2915 -6.5844>, rot=<0.1431 0.0988 0.3402 -0.9241>}
 AD_Lg_pine_rowF_T1 {parent=dyn_skirt_ba_01, loc=<23.3163 -0.1088 -4.0565>, rot=<-0.3176 -0.1943 0.2719 -0.8874>}
 AD_Lg_pine_rowF_V1 {parent=dyn_skirt_ba_01, loc=<21.9457 0.7797 -8.0685>, rot=<-0.3808 -0.2455 0.2812 -0.8459>}
 MSTR_skirtA_r {parent=pelvis, loc=<0.2547 9.8358 13.9059>, rot=<0.0547 -0.3847 0.9182 0.0765>}
 dyn_skirtA_r_01 {parent=MSTR_skirtA_r, loc=<0 0 0>, rot=<0 0 0 1>}
 dyn_skirtA_r_02 {parent=dyn_skirtA_r_01, loc=<8.3909 0 0>, rot=<0.0006 0.0034 0.008 -1>}
 dyn_skirtA_r_03 {parent=dyn_skirtA_r_02, loc=<8.0516 0 0>, rot=<0 0 0 -1>}
 dyn_skirtA_r_04 {parent=dyn_skirtA_r_03, loc=<9.138 0 0>, rot=<0 0 0 1>}
 AD_pine_rowG_F1 {parent=dyn_skirtA_r_01, loc=<16.9641 2.0453 0.7038>, rot=<0.2132 0.0834 0.3338 -0.9144>}
 AD_pine_rowG_F2 {parent=AD_pine_rowG_F1, loc=<3.811 0 0>, rot=<0 0 0 1>}
 AD_pine_rowG_H1 {parent=dyn_skirtA_r_01, loc=<15.2212 3.8488 -12.0634>, rot=<0.1196 0.0252 0.3135 -0.9417>}
 AD_pine_rowG_H2 {parent=AD_pine_rowG_H1, loc=<4.2935 0 0>, rot=<0 0 0 1>}
 AD_pine_rowG_G1 {parent=dyn_skirtA_r_01, loc=<15.5465 2.6681 -7.358>, rot=<0.2137 0.0947 0.3704 -0.899>}
 AD_pine_rowG_G2 {parent=AD_pine_rowG_G1, loc=<4.1208 0 0>, rot=<0 0 0 1>}
 AD_pine_rowE_B1 {parent=dyn_skirtA_r_01, loc=<0.3201 3.2729 4.2596>, rot=<0.1927 0.0736 0.3442 -0.916>}
 AD_pine_rowE_B2 {parent=AD_pine_rowE_B1, loc=<-0.6939 0 0>, rot=<0 0 0 1>}
 AD_pine_rowG_E1 {parent=dyn_skirtA_r_01, loc=<6.1238 1.3656 1.2876>, rot=<0.3648 0.1199 0.2068 -0.8999>}
 AD_pine_rowG_E2 {parent=AD_pine_rowG_E1, loc=<0.9298 0 0>, rot=<0 0 0 1>}
 AD_pine_rowA_W1 {parent=dyn_skirtA_r_01, loc=<6.092 2.395 -3.4731>, rot=<0.1836 0.0434 0.2365 -0.9531>}
 AD_pine_rowA_W2 {parent=AD_pine_rowA_W1, loc=<0.8716 0 0>, rot=<0.9699 0 0 0.2435>}
 AD_pine_rowE_C1 {parent=dyn_skirtA_r_01, loc=<-1.2023 1.3203 -0.3368>, rot=<0.1805 0.0685 0.354 -0.9151>}
 AD_pine_rowE_C2 {parent=AD_pine_rowE_C1, loc=<1.4288 0 0>, rot=<0 0 0 1>}
 AD_pine_rowE_D1 {parent=dyn_skirtA_r_01, loc=<-2.3735 2.695 -8.2027>, rot=<0.1462 0.0495 0.3661 -0.9177>}
 AD_pine_rowE_D2 {parent=AD_pine_rowE_D1, loc=<2.2346 0 0>, rot=<0 0 0 1>}
 AD_Lg_pine_rowG_F1 {parent=dyn_skirtA_r_01, loc=<22.8882 -2.4163 2.9922>, rot=<0.2132 0.0834 0.3338 -0.9144>}
 AD_Lg_pine_rowG_H1 {parent=dyn_skirtA_r_01, loc=<21.4466 -0.6876 -11.1138>, rot=<0.1196 0.0252 0.3135 -0.9417>}
 AD_Lg_pine_rowG_G1 {parent=dyn_skirtA_r_01, loc=<21.0384 -2.1867 -4.8073>, rot=<0.2137 0.0947 0.3704 -0.899>}
 AD_Sm_pine_rowE_B1 {parent=dyn_skirtA_r_01, loc=<6.1584 -1.4005 6.3359>, rot=<0.1927 0.0736 0.3442 -0.916>}
 AD_Med_pine_rowG_E1 {parent=dyn_skirtA_r_01, loc=<12.9977 -0.8445 4.1339>, rot=<0.3648 0.1199 0.2068 -0.8999>}
 AD_Sm_pine_rowA_W1 {parent=dyn_skirtA_r_01, loc=<12.9557 -0.9804 -2.1564>, rot=<0.1836 0.0434 0.2365 -0.9531>}
 AD_Med_pine_rowE_C1 {parent=dyn_skirtA_r_01, loc=<4.5404 -3.5168 1.6281>, rot=<0.1805 0.0685 0.354 -0.9151>}
 AD_Sm_pine_rowE_D1 {parent=dyn_skirtA_r_01, loc=<3.2687 -2.4083 -6.6667>, rot=<0.1462 0.0495 0.3661 -0.9177>}
 MSTR_skirtB_r {parent=pelvis, loc=<-0.1997 -2.8402 19.6659>, rot=<0.0908 -0.7013 0.7039 0.0676>}
 dyn_skirtB_r_01 {parent=MSTR_skirtB_r, loc=<0 0 0>, rot=<0 0 0 1>}
 dyn_skirtB_r_02 {parent=dyn_skirtB_r_01, loc=<8.4994 0 0>, rot=<-0.0008 0.0035 0.0111 -0.9999>}
 dyn_skirtB_r_03 {parent=dyn_skirtB_r_02, loc=<8.1774 0 0>, rot=<0 0 0 1>}
 dyn_skirtB_r_04 {parent=dyn_skirtB_r_03, loc=<9.2808 0.0001 0>, rot=<0 0 0 1>}
 AD_pine_rowF_Y1 {parent=dyn_skirtB_r_01, loc=<3.071 2.62 7.0354>, rot=<0.2034 0.0579 0.1992 -0.9569>}
 AD_pine_rowF_Y2 {parent=AD_pine_rowF_Y1, loc=<2.1051 0 0>, rot=<0 0 0 1>}
 AD_pine_rowD_P1 {parent=dyn_skirtB_r_01, loc=<-7.9147 2.5403 5.6421>, rot=<-0.1262 -0.098 0.2224 -0.9618>}
 AD_pine_rowD_P2 {parent=AD_pine_rowD_P1, loc=<9.0113 0 0>, rot=<0 0 0 1>}
 AD_pine_rowG_C1 {parent=dyn_skirtB_r_01, loc=<10.3504 -0.2423 0.2768>, rot=<0.0519 -0.0131 0.2226 -0.9734>}
 AD_pine_rowG_C2 {parent=AD_pine_rowG_C1, loc=<5.0461 0 0>, rot=<0 0 0 1>}
 AD_pine_rowG_B1 {parent=dyn_skirtB_r_01, loc=<18.353 1.2755 6.9278>, rot=<0.1051 0.0332 0.3815 -0.9178>}
 AD_pine_rowG_B2 {parent=AD_pine_rowG_B1, loc=<0.7702 0 0>, rot=<0 0 0 1>}
 AD_pine_rowE_A1 {parent=dyn_skirtB_r_01, loc=<1.3604 2.0739 -6.0746>, rot=<0.0261 -0.0258 0.2096 -0.9771>}
 AD_pine_rowE_A2 {parent=AD_pine_rowE_A1, loc=<-0.4638 0 0>, rot=<0 0 0 1>}
 AD_pine_rowG_D1 {parent=dyn_skirtB_r_01, loc=<15.7969 6.0554 -6.371>, rot=<-0.0935 -0.101 0.3431 -0.9292>}
 AD_pine_rowG_D2 {parent=AD_pine_rowG_D1, loc=<2.5439 0 0>, rot=<0 0 0 1>}
 AD_Sm_pine_rowF_Y1 {parent=dyn_skirtB_r_01, loc=<10.1641 -0.1564 8.5243>, rot=<0.2034 0.0579 0.1992 -0.9569>}
 AD_Lg_pine_rowD_P1 {parent=dyn_skirtB_r_01, loc=<-1.0705 -0.5881 3.7428>, rot=<-0.1262 -0.098 0.2224 -0.9618>}
 AD_Lg_pine_rowG_C1 {parent=dyn_skirtB_r_01, loc=<17.3397 -3.6166 0.2582>, rot=<0.0519 -0.0131 0.2226 -0.9734>}
 AD_Lg_pine_rowG_B1 {parent=dyn_skirtB_r_01, loc=<23.8384 -4.1048 8.0235>, rot=<0.1051 0.0332 0.3815 -0.9178>}
 AD_Sm_pine_rowE_A1 {parent=dyn_skirtB_r_01, loc=<8.4293 -1.1157 -6.3815>, rot=<0.0261 -0.0258 0.2096 -0.9771>}
 AD_Lg_pine_rowG_D1 {parent=dyn_skirtB_r_01, loc=<21.5723 1.253 -8.325>, rot=<-0.0935 -0.101 0.3431 -0.9292>}
 MSTR_skirtC_r {parent=pelvis, loc=<-0.6541 -15.5163 13.9059>, rot=<0.1052 -0.9155 0.3862 0.0412>}
 dyn_skirtC_r_01 {parent=MSTR_skirtC_r, loc=<0 0 0>, rot=<0 0 0 1>}
 dyn_skirtC_r_02 {parent=dyn_skirtC_r_01, loc=<8.5391 0 0>, rot=<-0.0016 0.0016 0.0124 -0.9999>}
 dyn_skirtC_r_03 {parent=dyn_skirtC_r_02, loc=<8.2233 0 0>, rot=<0 0 0 -1>}
 dyn_skirtC_r_04 {parent=dyn_skirtC_r_03, loc=<9.3329 0 0>, rot=<0 0 0 1>}
 AD_pine_rowE_N1 {parent=dyn_skirtC_r_01, loc=<-5.3271 4.868 7.0876>, rot=<0.0292 0.0003 0.3217 -0.9464>}
 AD_pine_rowE_N2 {parent=AD_pine_rowE_N1, loc=<8.7968 0 0>, rot=<0 0 0 1>}
 AD_pine_rowF_W1 {parent=dyn_skirtC_r_01, loc=<7.2723 2.3228 4.7307>, rot=<0.0284 -0.0034 0.2085 -0.9776>}
 AD_pine_rowF_W2 {parent=AD_pine_rowF_W1, loc=<5.1261 0 0>, rot=<0 0 0 1>}
 AD_pine_rowF_X1 {parent=dyn_skirtC_r_01, loc=<21.9212 -1.1888 4.1366>, rot=<0.0605 0.0112 0.1874 -0.9804>}
 AD_pine_rowF_X2 {parent=AD_pine_rowF_X1, loc=<-3.5498 0 0>, rot=<0 0 0 1>}
 AD_pine_rowF_Z1 {parent=dyn_skirtC_r_01, loc=<20.7444 0.6706 -0.9054>, rot=<-0.0629 -0.0644 0.362 -0.9278>}
 AD_pine_rowF_Z2 {parent=AD_pine_rowF_Z1, loc=<-1.1648 0 0>, rot=<0 0 0 1>}
 AD_pine_rowG_A1 {parent=dyn_skirtC_r_01, loc=<13.7968 1.9287 -4.7043>, rot=<-0.0601 -0.0605 0.3446 -0.9349>}
 AD_pine_rowG_A2 {parent=AD_pine_rowG_A1, loc=<5.3969 0 0>, rot=<0 0 0 1>}
 AD_pine_rowE_O1 {parent=dyn_skirtC_r_01, loc=<4.7516 0.9952 -0.0963>, rot=<0.019 -0.007 0.2832 -0.9589>}
 AD_pine_rowE_O2 {parent=AD_pine_rowE_O1, loc=<4.7621 0 0>, rot=<0 0 0 1>}
 AD_pine_rowD_O1 {parent=dyn_skirtC_r_01, loc=<-7.085 1.4877 0.6148>, rot=<-0.1731 -0.1098 0.2344 -0.9503>}
 AD_pine_rowD_O2 {parent=AD_pine_rowD_O1, loc=<8.1603 0 0>, rot=<0 0 0 1>}
 AD_Med_pine_rowE_N1 {parent=dyn_skirtC_r_01, loc=<0.8276 0.142 7.238>, rot=<0.0292 0.0003 0.3217 -0.9464>}
 AD_Lg_pine_rowF_W1 {parent=dyn_skirtC_r_01, loc=<14.3585 -0.8428 4.771>, rot=<0.0284 -0.0034 0.2085 -0.9776>}
 AD_Lg_pine_rowF_X1 {parent=dyn_skirtC_r_01, loc=<29.1354 -4.03 4.4832>, rot=<0.0605 0.0112 0.1874 -0.9804>}
 AD_Lg_pine_rowF_Z1 {parent=dyn_skirtC_r_01, loc=<26.4076 -4.4796 -2.1858>, rot=<-0.0629 -0.0644 0.362 -0.9278>}
 AD_Lg_pine_rowG_A1 {parent=dyn_skirtC_r_01, loc=<19.6581 -3.0153 -5.9036>, rot=<-0.0601 -0.0605 0.3446 -0.9349>}
 AD_Lg_pine_rowE_O1 {parent=dyn_skirtC_r_01, loc=<11.2675 -3.2214 -0.1165>, rot=<0.019 -0.007 0.2832 -0.9589>}
 AD_Med_pine_rowD_O1 {parent=dyn_skirtC_r_01, loc=<-0.3633 -1.6744 -1.634>, rot=<-0.1731 -0.1098 0.2344 -0.9503>}
 AD_pine_rowE_H1 {parent=pelvis, loc=<5.6459 10.9554 -0.0511>, rot=<-0.1751 0.3672 0.8099 0.4225>}
 AD_pine_rowD_E1 {parent=pelvis, loc=<11.4837 10.5644 5.556>, rot=<-0.0756 0.2269 0.9096 0.3397>}
 AD_pine_rowD_D1 {parent=pelvis, loc=<11.2972 10.6016 8.4447>, rot=<-0.0425 0.1459 0.9385 0.3099>}
 AD_pine_rowE_J1 {parent=pelvis, loc=<10.0163 8.5201 -9.0883>, rot=<-0.1728 0.5325 0.7784 0.2839>}
 AD_pine_rowE_K1 {parent=pelvis, loc=<7.8447 3.231 -15.1372>, rot=<-0.2402 0.6171 0.6881 0.2966>}
 AD_pine_rowD_Q1 {parent=pelvis, loc=<9.0558 -2.0394 15.4896>, rot=<0.1623 -0.5379 0.7829 0.2672>}
 AD_Sm_pine_rowE_H1 {parent=pelvis, loc=<1.1313 15.2688 -4.6609>, rot=<-0.1751 0.3672 0.8099 0.4225>}
 AD_Sm_pine_rowD_E1 {parent=pelvis, loc=<5.6021 15.0944 3.2927>, rot=<-0.0756 0.2269 0.9096 0.3397>}
 AD_Sm_pine_rowD_D1 {parent=pelvis, loc=<5.0545 15.0201 7.1242>, rot=<-0.0425 0.1459 0.9385 0.3099>}
 AD_Med_pine_rowE_J1 {parent=pelvis, loc=<3.9698 10.5217 -13.5234>, rot=<-0.1728 0.5325 0.7784 0.2839>}
 AD_Med_pine_rowE_K1 {parent=pelvis, loc=<2.345 4.098 -20.5445>, rot=<-0.2402 0.6171 0.6881 0.2966>}
 AD_Sm_pine_rowD_Q1 {parent=pelvis, loc=<2.8119 -0.1477 19.6935>, rot=<0.1623 -0.5379 0.7829 0.2672>}
]
