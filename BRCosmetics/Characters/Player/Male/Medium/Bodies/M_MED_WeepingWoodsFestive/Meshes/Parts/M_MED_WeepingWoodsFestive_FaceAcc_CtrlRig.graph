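# Control Rig for the WeepingWoodsFestive player cosmetic's face accessory, posing corrective CR_ bones that carry head, face and shoulder pieces as the head and arms move.
# A six-branch sequence runs spherical pose readers: one on upperarm_l, three on upperarm_r, two on head.
# Each reader's output weights Modify Transforms nodes holding preset bone offsets.
# Two further upperarm_l readers and a full-weight pair of Modify Transforms are unconnected to execution.
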ControlRigBlueprint /BRCosmetics/Characters/Player/Male/Medium/Bodies/M_MED_WeepingWoodsFestive/Meshes/Parts/M_MED_WeepingWoodsFestive_FaceAcc_CtrlRig
graphs [
 RigVMFunctionLibrary {links [], nodes []}
 RigVMFunction_Sequence_1_ContainedGraph {nodes [
   n1 = RigVMFunctionEntryNode()
   RigVMFunction_Sequence = &RigVMFunction_Sequence(ExecuteContext=@n1.ExecuteContext)
   n2 = &RigVMFunction_Sequence(ExecuteContext=@RigVMFunction_Sequence.B)
   n3 = &RigVMFunction_Sequence(ExecuteContext=@n2.B)
   n4 = &RigVMFunction_Sequence(ExecuteContext=@n3.B)
   n5 = &RigVMFunction_Sequence(ExecuteContext=@n4.B)
   RigVMFunctionReturnNode(A=@RigVMFunction_Sequence.A, B=@n2.A, C=@n3.A, D=@n4.A, E=@n5.A, F=@n5.B)
  ]}
 RigVMModel {nodes [
   &RigUnit_ModifyTransforms(
    ExecutePin=()
    ItemToModify="((Item=(Type=Bone,Name=\"None\"),Transform=(Rotation=(X=0,Y=0,Z=0,W=1),Translation=(X=0,Y=0,Z=0),Scale3D=(X=1,Y=1,Z=1))),(Item=(Type=Bone,Name=\"CR_spineRot_05\"),Transform=(Rotation=(X=-0.0886,Y=0.0136,Z=0.1515,W=0.9844),Translation=(X=3.501,Y=-3.159,Z=0.5732 ...(truncated)"
    Weight=1
    WeightMinimum=0
   )
   > &RigUnit_ModifyTransforms(
    ItemToModify="((Item=(Type=Bone,Name=\"None\"),Transform=(Rotation=(X=0,Y=0,Z=0,W=1),Translation=(X=0,Y=0,Z=0),Scale3D=(X=1,Y=1,Z=1))),(Item=(Type=Bone,Name=\"CR_headBa_E_bt\"),Transform=(Rotation=(X=0,Y=0,Z=-0.112,W=0.9937),Translation=(X=0,Y=0,Z=0),Scale3D=(X=1,Y=1,Z=1))) ...(truncated)"
    Weight=1
    WeightMinimum=0
   )
   RigUnit_BeginExecution = &RigUnit_BeginExecution()
   RigVMFunction_Sequence = RigVMAggregateNode(ExecuteContext=@RigUnit_BeginExecution.ExecutePin)
   n1 = &RigUnit_SphericalPoseReader(
    ActiveRegionScaleFactors=(PositiveWidth=2.5,NegativeWidth=3,PositiveHeight=3,NegativeHeight=3)
    ActiveRegionSize=0.1
    Debug=(bDrawDebug=false,DebugScale=25,DebugSegments=20,DebugThickness=0.25)
    DriverAxis=(X=1,Y=0,Z=0)
    DriverItem=(Type=Bone,Name="upperarm_l")
    FalloffRegionScaleFactors=(PositiveWidth=1,NegativeWidth=1,PositiveHeight=1,NegativeHeight=1)
    FalloffSize=0.5
    OptionalParentItem=(Type=Bone)
    RotationOffset=(X=0,Y=0,Z=0)
    ExecutePin=@RigVMFunction_Sequence.A
   )
   > &RigUnit_ModifyTransforms(
    ItemToModify="((Item=(Type=Bone,Name=\"None\"),Transform=(Rotation=(X=0,Y=0,Z=0,W=1),Translation=(X=0,Y=0,Z=0),Scale3D=(X=1,Y=1,Z=1))),(Item=(Type=Bone,Name=\"CR_shoulder_l_A\"),Transform=(Rotation=(X=0,Y=0,Z=0,W=1),Translation=(X=0,Y=0,Z=0),Scale3D=(X=1,Y=1,Z=1))),(Item=(T ...(truncated)"
    WeightMinimum=0
    Weight=@n1.OutputParam
   )
   n2 = &RigUnit_SphericalPoseReader(
    ActiveRegionScaleFactors=(PositiveWidth=2.5,NegativeWidth=3,PositiveHeight=3,NegativeHeight=3)
    ActiveRegionSize=0.1
    Debug=(bDrawDebug=false,DebugScale=25,DebugSegments=20,DebugThickness=0.25)
    DriverAxis=(X=-1,Y=0,Z=0)
    DriverItem=(Type=Bone,Name="upperarm_r")
    FalloffRegionScaleFactors=(PositiveWidth=1,NegativeWidth=1,PositiveHeight=1,NegativeHeight=1)
    FalloffSize=0.55
    FlipHeightScaling=true
    OptionalParentItem=(Type=Bone)
    RotationOffset=(X=180,Y=0,Z=180)
    ExecutePin=@RigVMFunction_Sequence.B
   )
   > &RigUnit_ModifyTransforms(
    ItemToModify="((Item=(Type=Bone,Name=\"CR_shoulder_r_C\"),Transform=(Rotation=(X=-0.0167,Y=0.1157,Z=0.1418,W=0.983),Translation=(X=0,Y=0,Z=0),Scale3D=(X=1,Y=1,Z=1))),(Item=(Type=Bone,Name=\"CR_shoulder_r_B\"),Transform=(Rotation=(X=-0.3566,Y=-0.0157,Z=-0.0585,W=0.9323),Tran ...(truncated)"
    WeightMinimum=0
    Weight=@n2.OutputParam
   )
   n3 = &RigUnit_SphericalPoseReader(
    ActiveRegionScaleFactors=(PositiveWidth=3,NegativeWidth=3,PositiveHeight=3,NegativeHeight=3)
    ActiveRegionSize=0.1
    Debug=(bDrawDebug=false,bDraw2D=False,bDrawLocalAxes=False,DebugScale=25,DebugSegments=20,DebugThickness=0.25)
    DriverAxis=(X=1,Y=0,Z=0)
    DriverItem=(Type=Bone,Name="head")
    FalloffRegionScaleFactors=(PositiveWidth=2,NegativeWidth=2,PositiveHeight=2,NegativeHeight=2)
    FalloffSize=0.4
    FlipHeightScaling=False
    FlipWidthScaling=False
    OptionalParentItem=(Type=Bone,Name="None")
    RotationOffset=(X=90,Y=0,Z=0)
    ExecutePin=@RigVMFunction_Sequence.C
   )
   > &RigUnit_ModifyTransforms(
    ItemToModify="((Item=(Type=Bone,Name=\"None\"),Transform=(Rotation=(X=0,Y=0,Z=0,W=1),Translation=(X=0,Y=0,Z=0),Scale3D=(X=1,Y=1,Z=1))),(Item=(Type=Bone,Name=\"CR_head_spike_fr_r\"),Transform=(Rotation=(X=0.0178,Y=0.075,Z=0.0751,W=0.9942),Translation=(X=0,Y=0,Z=0),Scale3D=(X ...(truncated)"
    WeightMinimum=0
    Weight=@n3.OutputParam
   )
   > &RigUnit_ModifyTransforms(
    ItemToModify="((Item=(Type=Bone,Name=\"None\"),Transform=(Rotation=(X=0,Y=0,Z=0,W=1),Translation=(X=0,Y=0,Z=0),Scale3D=(X=1,Y=1,Z=1))),(Item=(Type=Bone,Name=\"CR_head_spikes_l\"),Transform=(Rotation=(X=0,Y=0,Z=0,W=1),Translation=(X=0,Y=0,Z=0),Scale3D=(X=1,Y=1,Z=1))),(Item=( ...(truncated)"
    WeightMinimum=0
    Weight=@n3.OutputParam
   )
   > &RigUnit_ModifyTransforms(
    ItemToModify="((Item=(Type=Bone,Name=\"CR_head_spikes_l1\"),Transform=(Rotation=(X=0.0376,Y=-0.2118,Z=0.1707,W=0.9616),Translation=(X=0,Y=0,Z=0),Scale3D=(X=1,Y=1,Z=1))),(Item=(Type=Bone,Name=\"CR_head_spike_fr_l\"),Transform=(Rotation=(X=0,Y=0,Z=0,W=1),Translation=(X=0,Y=0, ...(truncated)"
    WeightMinimum=0
    Weight=@n3.OutputParam
   )
   > &RigUnit_ModifyTransforms(
    ItemToModify="((Item=(Type=Bone,Name=\"None\"),Transform=(Rotation=(X=0,Y=0,Z=0,W=1),Translation=(X=0,Y=0,Z=0),Scale3D=(X=1,Y=1,Z=1))),(Item=(Type=Bone,Name=\"CR_head_spikes_r1\"),Transform=(Rotation=(X=-0.0001,Y=-0.2235,Z=0,W=0.9747),Translation=(X=0,Y=0,Z=0),Scale3D=(X=1, ...(truncated)"
    WeightMinimum=0
    Weight=@n3.OutputParam
   )
   n4 = &RigUnit_SphericalPoseReader(
    ActiveRegionScaleFactors=(PositiveWidth=2,NegativeWidth=2,PositiveHeight=2,NegativeHeight=2)
    ActiveRegionSize=0.1
    Debug=(bDrawDebug=false,bDraw2D=False,bDrawLocalAxes=False,DebugScale=25,DebugSegments=20,DebugThickness=0.25)
    DriverAxis=(X=1,Y=0,Z=0)
    DriverItem=(Type=Bone,Name="head")
    FalloffRegionScaleFactors=(PositiveWidth=1,NegativeWidth=1,PositiveHeight=1,NegativeHeight=1)
    FalloffSize=0.43
    FlipHeightScaling=False
    FlipWidthScaling=False
    OptionalParentItem=(Type=Bone,Name="None")
    RotationOffset=(X=-90,Y=0,Z=8)
    ExecutePin=@RigVMFunction_Sequence.D
   )
   > &RigUnit_ModifyTransforms(
    ItemToModify="((Item=(Type=Bone,Name=\"None\"),Transform=(Rotation=(X=0,Y=0,Z=0,W=1),Translation=(X=0,Y=0,Z=0),Scale3D=(X=1,Y=1,Z=1))),(Item=(Type=Bone,Name=\"CR_head_spike_fr_r\"),Transform=(Rotation=(X=0.0086,Y=-0.1219,Z=0.1362,W=0.9831),Translation=(X=0,Y=0,Z=0),Scale3D= ...(truncated)"
    WeightMinimum=0
    Weight=@n4.OutputParam
   )
   > &RigUnit_ModifyTransforms(
    ItemToModify="((Item=(Type=Bone,Name=\"None\"),Transform=(Rotation=(X=0,Y=0,Z=0,W=1),Translation=(X=0,Y=0,Z=0),Scale3D=(X=1,Y=1,Z=1))),(Item=(Type=Bone,Name=\"CR_head_spikes_l\"),Transform=(Rotation=(X=0,Y=0,Z=0.2876,W=0.9578),Translation=(X=0,Y=0,Z=0),Scale3D=(X=1,Y=1,Z=1) ...(truncated)"
    WeightMinimum=0
    Weight=@n4.OutputParam
   )
   > &RigUnit_ModifyTransforms(
    ItemToModify="((Item=(Type=Bone,Name=\"CR_head_spikes_l1\"),Transform=(Rotation=(X=0,Y=0.2258,Z=0,W=0.9742),Translation=(X=0,Y=0,Z=0),Scale3D=(X=1,Y=1,Z=1))),(Item=(Type=Bone,Name=\"CR_head_spike_fr_l\"),Transform=(Rotation=(X=0,Y=-0.0511,Z=0,W=0.9987),Translation=(X=0,Y=0, ...(truncated)"
    WeightMinimum=0
    Weight=@n4.OutputParam
   )
   > &RigUnit_ModifyTransforms(
    ItemToModify="((Item=(Type=Bone,Name=\"None\"),Transform=(Rotation=(X=0,Y=0,Z=0,W=1),Translation=(X=0,Y=0,Z=0),Scale3D=(X=1,Y=1,Z=1))),(Item=(Type=Bone,Name=\"CR_head_spikes_r1\"),Transform=(Rotation=(X=-0.0248,Y=0,Z=0,W=0.9997),Translation=(X=0,Y=0,Z=0),Scale3D=(X=1,Y=1,Z= ...(truncated)"
    WeightMinimum=0
    Weight=@n4.OutputParam
   )
   n5 = &RigUnit_SphericalPoseReader(
    ActiveRegionScaleFactors=(PositiveWidth=0.5,NegativeWidth=3,PositiveHeight=1.5,NegativeHeight=1.5)
    ActiveRegionSize=0.1
    Debug=(bDrawDebug=false,DebugScale=25,DebugSegments=20,DebugThickness=0.25)
    DriverAxis=(X=-1,Y=0,Z=0)
    DriverItem=(Type=Bone,Name="upperarm_r")
    FalloffRegionScaleFactors=(PositiveWidth=1,NegativeWidth=1,PositiveHeight=1,NegativeHeight=1)
    FalloffSize=0.5
    FlipWidthScaling=true
    OptionalParentItem=(Type=Bone)
    RotationOffset=(X=90,Y=180,Z=180)
    ExecutePin=@RigVMFunction_Sequence.E
   )
   > &RigUnit_ModifyTransforms(
    ItemToModify="((Item=(Type=Bone,Name=\"CR_shoulder_r_A\"),Transform=(Rotation=(X=0.0003,Y=-0.011,Z=0.0306,W=0.9995),Translation=(X=0,Y=0,Z=0),Scale3D=(X=1,Y=1,Z=1))),(Item=(Type=Bone,Name=\"CR_shoulder_r_B\"),Transform=(Rotation=(X=0.0741,Y=-0.0325,Z=0.2042,W=0.9756),Transl ...(truncated)"
    WeightMinimum=0
    Weight=@n5.OutputParam
   )
   n6 = &RigUnit_SphericalPoseReader(
    ActiveRegionScaleFactors=(PositiveWidth=3,NegativeWidth=1,PositiveHeight=1.5,NegativeHeight=1.5)
    ActiveRegionSize=0.1
    Debug=(bDrawDebug=false,DebugScale=25,DebugSegments=20,DebugThickness=0.25)
    DriverAxis=(X=-1,Y=0,Z=0)
    DriverItem=(Type=Bone,Name="upperarm_r")
    FalloffRegionScaleFactors=(PositiveWidth=1,NegativeWidth=1,PositiveHeight=1,NegativeHeight=1)
    FalloffSize=0.4
    FlipHeightScaling=true
    FlipWidthScaling=false
    OptionalParentItem=(Type=Bone)
    RotationOffset=(X=-50,Y=0,Z=0)
    ExecutePin=@RigVMFunction_Sequence.F
   )
   > &RigUnit_ModifyTransforms(
    ItemToModify="((Item=(Type=Bone,Name=\"CR_shoulder_r_A\"),Transform=(Rotation=(X=0.1127,Y=-0.0234,Z=0.2016,W=0.9727),Translation=(X=0,Y=0,Z=0),Scale3D=(X=1,Y=1,Z=1))),(Item=(Type=Bone,Name=\"CR_shoulder_r_B\"),Transform=(Rotation=(X=0,Y=0,Z=0.0055,W=1),Translation=(X=0,Y=0, ...(truncated)"
    WeightMinimum=0
    Weight=@n6.OutputParam
   )
   &RigUnit_SphericalPoseReader(
    ActiveRegionScaleFactors=(PositiveWidth=0.5,NegativeWidth=3,PositiveHeight=1.5,NegativeHeight=1.5)
    ActiveRegionSize=0.1
    Debug=(bDrawDebug=false,DebugScale=25,DebugSegments=20,DebugThickness=0.25)
    DriverAxis=(X=1,Y=0,Z=0)
    DriverItem=(Type=Bone,Name="upperarm_l")
    ExecutePin=()
    FalloffRegionScaleFactors=(PositiveWidth=1,NegativeWidth=1,PositiveHeight=1,NegativeHeight=1)
    FalloffSize=0.4
    OptionalParentItem=(Type=Bone)
    RotationOffset=(X=90,Y=0,Z=0)
   )
   &RigUnit_SphericalPoseReader(
    ActiveRegionScaleFactors=(PositiveWidth=3,NegativeWidth=1,PositiveHeight=1.5,NegativeHeight=1.5)
    ActiveRegionSize=0.1
    Debug=(bDrawDebug=false,DebugScale=25,DebugSegments=20,DebugThickness=0.25)
    DriverAxis=(X=1,Y=0,Z=0)
    DriverItem=(Type=Bone,Name="upperarm_l")
    ExecutePin=()
    FalloffRegionScaleFactors=(PositiveWidth=1,NegativeWidth=1,PositiveHeight=1,NegativeHeight=1)
    FalloffSize=0.4
    OptionalParentItem=(Type=Bone)
    RotationOffset=(X=90,Y=0,Z=180)
   )
  ]}
]
hierarchy [
 root {loc=<0 0 0>, rot=<0 0 0 1>}
 pelvis {parent=root, loc=<0 -0.7071 93.1191>, rot=<0.0127 0.707 -0.0127 -0.707>}
 spine_01 {parent=pelvis, loc=<6.7494 -0.6388 0>, rot=<0 0 0.034 -0.9994>}
 spine_02 {parent=spine_01, loc=<7.2518 -0.035 0>, rot=<0 0 0.0238 -0.9997>}
 spine_03 {parent=spine_02, loc=<6.0104 0 0>, rot=<0 0 0.0149 -0.9999>}
 spine_04 {parent=spine_03, loc=<8.1024 0 0>, rot=<0 0 -0.0132 -0.9999>}
 spine_05 {parent=spine_04, loc=<10.0146 0 0>, rot=<0 0 -0.0015 -1>}
 neck_01 {parent=spine_05, loc=<11.3138 0.1529 0>, rot=<0 0 -0.2003 -0.9797>}
 neck_02 {parent=neck_01, loc=<4.5627 0 0>, rot=<0 0 0.0504 -0.9987>}
 head {parent=neck_02, loc=<5.3914 0 0>, rot=<0 0 0.115 -0.9934>}
 MSTR_hatAll {parent=head, loc=<12.4327 -3.3313 -0.0836>, rot=<1 0.0041 0 0>}
 dyn_hatAll_01 {parent=MSTR_hatAll, loc=<0 0 0>, rot=<0 0 0 1>}
 dyn_hatAll_02 {parent=dyn_hatAll_01, loc=<7.8272 0 0>, rot=<0 0 0 1>}
 MSTR_hat {parent=dyn_hatAll_02, loc=<2.5505 11.5543 -0.0815>, rot=<0.0015 -0.0006 0.9335 0.3586>}
 dyn_hat_01 {parent=MSTR_hat, loc=<0 0 0>, rot=<0 0 0 1>}
 dyn_hat_02 {parent=dyn_hat_01, loc=<10.9625 0 0>, rot=<-0.0001 -0.0026 -0.1753 -0.9845>}
 dyn_hat_03 {parent=dyn_hat_02, loc=<6.8063 0 0>, rot=<0.0002 -0.0017 -0.0551 -0.9985>}
 dyn_hat_04 {parent=dyn_hat_03, loc=<5.4293 0 0>, rot=<0 0 0 1>}
 AD_Sm_head_frontQ_01 {parent=head, loc=<13.8959 11.8203 0.8295>, rot=<-0.0716 0.4117 0.8964 0.1482>}
 AD_Med_head_frontR_01 {parent=head, loc=<12.7755 8.6005 -8.4635>, rot=<-0.2625 0.5675 0.711 0.3218>}
 AD_Med_head_frontP_01 {parent=head, loc=<12.5601 7.4851 10.06>, rot=<0.3059 -0.5866 0.6677 0.3413>}
 AD_Sm_face_A1 {parent=head, loc=<-1.5516 9.7541 12.2975>, rot=<0.4082 -0.4611 0.7846 -0.0728>}
 CR_head_spikes_l {parent=head, loc=<1.1131 -8.1453 -6.9463>, rot=<0.9397 0 0 0.342>}
 AD_Sm_headTp_A1 {parent=CR_head_spikes_l, loc=<10.4834 -3.3639 0.2285>, rot=<0.4113 -0.5495 0.6444 0.3371>}
 AD_Sm_headTp_B1 {parent=CR_head_spikes_l, loc=<6.8881 1.4808 3.879>, rot=<0.3508 -0.5562 0.7404 0.1391>}
 CR_head_spikes_ct {parent=head, loc=<4.4892 -9.0797 0>, rot=<0 0 0 1>}
 AD_Med_headTp_C1 {parent=CR_head_spikes_ct, loc=<2.1369 -5.1013 -4.6386>, rot=<0.2681 -0.8537 -0.3467 -0.2812>}
 AD_Sm_headTp_D1 {parent=CR_head_spikes_ct, loc=<1.9634 -6.348 0.1253>, rot=<0.2848 -0.9226 -0.0023 -0.2602>}
 AD_Sm_headTp_E1 {parent=CR_head_spikes_ct, loc=<1.746 -4.9354 4.8201>, rot=<0.3878 -0.8777 0.2133 0.1838>}
 CR_head_spikes_ct1 {parent=head, loc=<-0.847 -9.1716 0>, rot=<0 0 0 1>}
 CR_headBa_D {parent=CR_head_spikes_ct1, loc=<2.7211 -6.0621 2.8588>, rot=<0.3482 -0.9259 -0.0192 -0.1451>}
 AD_Med_headMid_D1 {parent=CR_headBa_D, loc=<0 0 0>, rot=<0 0 0 1>}
 CR_headBa_K {parent=CR_head_spikes_ct1, loc=<-0.9366 -6.071 -0.2969>, rot=<0.3134 -0.9354 -0.1135 -0.1179>}
 AD_Med_headMid_K1 {parent=CR_headBa_K, loc=<0 0 0>, rot=<0 0 0 1>}
 CR_headBa_E {parent=CR_head_spikes_ct1, loc=<-0.702 -6.1181 6.8326>, rot=<0.2985 -0.8788 0.2442 0.281>}
 AD_Med_headMid_E1 {parent=CR_headBa_E, loc=<0 0 0>, rot=<0 0 0 1>}
 CR_headBa_L {parent=CR_head_spikes_ct1, loc=<-3.452 -6.8498 5.1893>, rot=<0.3235 -0.9187 0.2254 -0.0235>}
 AD_Lg_headMid_L1 {parent=CR_headBa_L, loc=<0 0 0>, rot=<0 0 0 1>}
 CR_headBa_J {parent=CR_head_spikes_ct1, loc=<0.0207 -5.7985 -4.6757>, rot=<0.3326 -0.8581 -0.3173 -0.2288>}
 AD_Lg_headMid_J1 {parent=CR_headBa_J, loc=<0 0 0>, rot=<0 0 0 1>}
 CR_head_spikes_r1 {parent=head, loc=<-0.5789 -6.1617 8.6107>, rot=<-0.342 0 0 -0.9397>}
 CR_headMid_G {parent=CR_head_spikes_r1, loc=<0.507 -0.7467 6.6996>, rot=<0.2718 -0.6326 0.6025 0.4037>}
 AD_Sm_headMid_G1 {parent=CR_headMid_G, loc=<0 0 0>, rot=<0 0 0 1>}
 CR_headMid_F {parent=CR_head_spikes_r1, loc=<2.3119 -5.5818 3.613>, rot=<0.1462 -0.6754 0.6109 0.3863>}
 AD_Sm_headMid_F1 {parent=CR_headMid_F, loc=<0 0 0>, rot=<0 0 0 1>}
 AD_Med_headTp_F1 {parent=CR_head_spikes_r1, loc=<5.8722 -2.5879 5.2894>, rot=<0.2703 -0.5538 0.6484 0.447>}
 CR_head_spikes_r4 {parent=head, loc=<1.4452 -0.6094 7.6317>, rot=<0.0872 0 0 -0.9962>}
 AD_Lg_head_l_Q1 {parent=CR_head_spikes_r4, loc=<1.1714 -1.6079 8.1487>, rot=<0.3665 -0.7915 0.4767 0.1094>}
 AD_Lg_headTp_G1 {parent=CR_head_spikes_r4, loc=<3.2284 -6.5861 6.7594>, rot=<0.3427 -0.7923 0.4905 0.1196>}
 CR_head_l_P {parent=CR_head_spikes_r4, loc=<-1.9691 -3.9769 7.423>, rot=<0.4322 -0.8188 0.3549 0.1298>}
 AD_Lg_headMid_P1 {parent=CR_head_l_P, loc=<0 0 0>, rot=<0 0 0 1>}
 CR_head_spikes_l1 {parent=head, loc=<-0.5789 -6.1617 -8.6107>, rot=<0.9397 0 0 0.342>}
 CR_head_l_A {parent=CR_head_spikes_l1, loc=<1.8249 -9.5279 3.4013>, rot=<0.1726 -0.7801 0.4658 0.3803>}
 AD_Lg_headMid_A1 {parent=CR_head_l_A, loc=<0 0 0>, rot=<0 0 0 1>}
 CR_headMid_C {parent=CR_head_spikes_l1, loc=<2.1911 2.6494 5.5859>, rot=<0.2773 -0.5328 0.7547 0.2637>}
 AD_Lg_headMid_C1 {parent=CR_headMid_C, loc=<0 0 0>, rot=<0 0 0 1>}
 CR_head_l_B {parent=CR_head_spikes_l1, loc=<4.5306 -4.122 6.7039>, rot=<0.2815 -0.7609 0.5124 0.2815>}
 AD_Lg_headMid_B1 {parent=CR_head_l_B, loc=<0 0 0>, rot=<0 0 0 1>}
 CR_head_l_H {parent=CR_head_spikes_l1, loc=<-2.6897 -2.7721 6.5648>, rot=<0.2669 -0.7153 0.5837 0.2762>}
 AD_Med_headMid_H1 {parent=CR_head_l_H, loc=<0 0 0>, rot=<0 0 0 1>}
 CR_headBa_I {parent=CR_head_spikes_l1, loc=<-3.558 6.6094 5.9417>, rot=<0.3217 -0.5711 0.7019 0.2787>}
 AD_Lg_headMid_I1 {parent=CR_headBa_I, loc=<0 0 0>, rot=<0 0 0 1>}
 CR_head_spike_fr_r {parent=head, loc=<-5.6317 5.5802 7.5311>, rot=<0.4226 0 0 -0.9063>}
 AD_Med_face_C1 {parent=CR_head_spike_fr_r, loc=<-0.3711 5.1592 5.1143>, rot=<0.4576 -0.5416 0.7051 -0.0118>}
 AD_Med_face_B1 {parent=CR_head_spike_fr_r, loc=<0.46 0.947 6.3021>, rot=<0.4714 -0.6418 0.595 -0.1091>}
 CR_head_spike_fr_ct {parent=head, loc=<-5.6317 5.5802 0>, rot=<0 0 0 1>}
 AD_Med_face_D1 {parent=CR_head_spike_fr_ct, loc=<-1.0639 8.3862 3.3319>, rot=<0.2847 -0.2389 0.88 0.2959>}
 AD_Med_face_G1 {parent=CR_head_spike_fr_ct, loc=<-0.1625 8.4382 -1.3249>, rot=<-0.1215 0.054 0.9676 0.2147>}
 AD_Med_face_H1 {parent=CR_head_spike_fr_ct, loc=<-0.1339 7.4206 -5.5868>, rot=<-0.2774 0.1987 0.9176 0.2039>}
 CR_head_spike_fr_l {parent=head, loc=<-5.6317 5.5802 -7.0339>, rot=<0.9397 0 0 0.342>}
 AD_Sm_face_I1 {parent=CR_head_spike_fr_l, loc=<-0.7038 -5.9184 -0.1248>, rot=<0.3707 -0.8991 -0.0675 0.2229>}
 AD_Sm_face_J1 {parent=CR_head_spike_fr_l, loc=<1.8765 -6.4322 3.4854>, rot=<0.2475 -0.8781 0.1246 0.3902>}
 CR_head_spike_fr_l2 {parent=head, loc=<-2.3095 2.3321 -12.3014>, rot=<-0.8192 0 0 -0.5736>}
 CR_face_l_J {parent=CR_head_spike_fr_l2, loc=<-0.72 -3.9163 1.5153>, rot=<0.1985 -0.9215 -0.0067 0.3338>}
 AD_Med_faceBt_J1 {parent=CR_face_l_J, loc=<0 0 0>, rot=<0 0 0 1>}
 CR_face_l_I {parent=CR_head_spike_fr_l2, loc=<-4.8427 -4.9195 -0.6929>, rot=<0.2977 -0.8518 -0.138 0.4084>}
 AD_Med_faceBt_I1 {parent=CR_face_l_I, loc=<0 0 0>, rot=<0 0 0 1>}
 CR_head_spike_fr_r2 {parent=head, loc=<-7.9905 3.5423 9.9598>, rot=<0.2588 0 0 -0.9659>}
 CR_faceBt_B {parent=CR_head_spike_fr_r2, loc=<-2.4395 3.1573 3.6393>, rot=<0.4381 -0.572 0.6889 0.0792>}
 AD_Sm_faceBt_B1 {parent=CR_faceBt_B, loc=<0 0 0>, rot=<0 0 0 1>}
 CR_face_l_M {parent=CR_head_spike_fr_r2, loc=<5.1486 -3.6297 5.7647>, rot=<0.345 -0.7968 0.4955 -0.0222>}
 AD_Med_faceBt_M1 {parent=CR_face_l_M, loc=<0 0 0>, rot=<0 0 0 1>}
 CR_faceBt_A {parent=CR_head_spike_fr_r2, loc=<0.8287 0.5089 4.6785>, rot=<0.5317 -0.6149 0.5768 -0.0808>}
 AD_Sm_faceBt_A1 {parent=CR_faceBt_A, loc=<0 0 0>, rot=<0 0 0 1>}
 CR_head_spikes_r2 {parent=neck_02, loc=<-4.5427 -3.9986 8.6107>, rot=<-0.3398 0.0393 0.108 -0.9335>}
 CR_headBa_M {parent=CR_head_spikes_r2, loc=<2.9785 -6.6986 3.6404>, rot=<0.2238 -0.7591 0.5823 0.1857>}
 AD_Sm_headMid_M1 {parent=CR_headBa_M, loc=<0 0 0>, rot=<0 0 0 1>}
 CR_headBa_N {parent=CR_head_spikes_r2, loc=<4.4534 -4.7092 4.3408>, rot=<0.246 -0.6926 0.563 0.378>}
 AD_Med_headMid_N1 {parent=CR_headBa_N, loc=<0 0 0>, rot=<0 0 0 1>}
 CR_head_l_O {parent=CR_head_spikes_r2, loc=<6.9179 0.8965 3.4856>, rot=<0.2049 -0.6906 0.5944 0.3575>}
 AD_Med_headMid_O1 {parent=CR_head_l_O, loc=<0 0 0>, rot=<0 0 0 1>}
 CR_head_spike_fr_ct1 {parent=neck_02, loc=<-5.3943 8.2619 0>, rot=<0 0 0.115 -0.9934>}
 CR_faceBt_D {parent=CR_head_spike_fr_ct1, loc=<2.1616 6.9929 1.2247>, rot=<0.1308 -0.056 0.9706 0.194>}
 AD_Sm_faceBt_D1 {parent=CR_faceBt_D, loc=<0 0 0>, rot=<0 0 0 1>}
 CR_faceBt_F {parent=CR_head_spike_fr_ct1, loc=<2.5913 6.6669 -4.2706>, rot=<-0.1952 0.1452 0.9605 0.135>}
 AD_Sm_faceBt_F1 {parent=CR_faceBt_F, loc=<0 0 0>, rot=<0 0 0 1>}
 CR_faceBt_C {parent=CR_head_spike_fr_ct1, loc=<2.0894 6.0387 7.0319>, rot=<0.2286 -0.2554 0.9387 0.0364>}
 AD_Lg_faceBt_C1 {parent=CR_faceBt_C, loc=<0 0 0>, rot=<0 0 0 1>}
 CR_head_spike_fr_r1 {parent=neck_01, loc=<-0.1427 8.267 8.1566>, rot=<0.4168 -0.0697 0.1495 -0.8939>}
 CR_headFr_B_bt {parent=CR_head_spike_fr_r1, loc=<-0.1232 -2.8332 3.4137>, rot=<0.3064 -0.749 0.5719 0.1348>}
 AD_Lg_headFr_B_bt1 {parent=CR_headFr_B_bt, loc=<0 0 0>, rot=<0 0 0 1>}
 CR_headFr_C_bt {parent=CR_head_spike_fr_r1, loc=<-0.0482 2.0786 3.1936>, rot=<0.3528 -0.8071 0.4287 0.2008>}
 AD_Med_headFr_C_bt1 {parent=CR_headFr_C_bt, loc=<0 0 0>, rot=<0 0 0 1>}
 CR_head_spike_fr_l1 {parent=neck_01, loc=<0.0813 5.9214 -9.3002>, rot=<0.9268 -0.155 -0.0564 0.3373>}
 AD_Sm_faceBt_H1 {parent=CR_head_spike_fr_l1, loc=<4.7101 -6.4233 -0.6081>, rot=<0.2739 -0.9219 -0.0174 0.2734>}
 CR_faceBt_G {parent=CR_head_spike_fr_l1, loc=<2.7316 -5.6694 -5.461>, rot=<0.2829 -0.92 -0.1196 0.2434>}
 AD_Sm_faceBt_G1 {parent=CR_faceBt_G, loc=<0 0 0>, rot=<0 0 0 1>}
 clavicle_l {parent=spine_05, loc=<4.3575 -0.1466 -2.3223>, rot=<0.0465 -0.7289 -0.009 -0.6829>}
 upperarm_l {parent=clavicle_l, loc=<15.5111 0 0>, rot=<0.0226 -0.3234 -0.0088 -0.9459>}
 clavicle_r {parent=spine_05, loc=<4.3575 -0.1466 2.3223>, rot=<0.7289 0.0465 -0.6829 0.009>}
 upperarm_r {parent=clavicle_r, loc=<-15.5111 0 0>, rot=<0.0226 -0.3234 -0.0088 -0.9459>}
 CR_head_spikes_r3 {parent=clavicle_r, loc=<-8.6992 0.3638 -3.4887>, rot=<0.7301 0.0202 -0.6828 0.0156>}
 CR_shoulder_r_A {parent=CR_head_spikes_r3, loc=<2.6541 -7.0506 1.3911>, rot=<0.3746 -0.7523 0.3924 0.3738>}
 AD_Lg_shoulder_A_r1 {parent=CR_shoulder_r_A, loc=<0 0 0>, rot=<0 0 0 1>}
 CR_shoulder_r_B {parent=CR_head_spikes_r3, loc=<5.0212 -3.7648 3.6859>, rot=<0.4378 -0.6682 0.4255 0.4252>}
 AD_Lg_shoulder_B_r1 {parent=CR_shoulder_r_B, loc=<0 0 0>, rot=<0 0 0 1>}
 CR_shoulder_r_C {parent=CR_head_spikes_r3, loc=<6.2531 2.6054 2.0298>, rot=<0.4814 -0.5012 0.6287 0.349>}
 AD_Lg_shoulder_C_r1 {parent=CR_shoulder_r_C, loc=<0 0 0>, rot=<0 0 0 1>}
 CR_shoulder_r_D {parent=CR_head_spikes_r3, loc=<2.751 5.6152 3.3281>, rot=<0.4602 -0.4133 0.7044 0.3482>}
 AD_Sm_shoulder_D_r1 {parent=CR_shoulder_r_D, loc=<0 0 0>, rot=<0 0 0 1>}
 CR_headFr_A_bt {parent=CR_head_spikes_r3, loc=<1.0763 6.8967 2.7152>, rot=<0.3124 -0.4417 0.8085 0.2314>}
 AD_Sm_headFr_A_bt1 {parent=CR_headFr_A_bt, loc=<0 0 0>, rot=<0 0 0 1>}
 CR_spineRot_05 {parent=spine_05, loc=<0 0 0>, rot=<0 0 0 1>}
 CR_headBa_A_bt {parent=CR_spineRot_05, loc=<12.2891 -8.7178 -7.1994>, rot=<0.1948 -0.9296 -0.1416 -0.2789>}
 AD_headBa_A_bt1 {parent=CR_headBa_A_bt, loc=<0 0 0>, rot=<0 0 0 1>}
 CR_head_spike_fr_ct2 {parent=spine_05, loc=<5.7389 8.0683 0>, rot=<0 0 -0.036 -0.9994>}
 CR_headFr_D_bt {parent=CR_head_spike_fr_ct2, loc=<2.4213 4.8614 3.4864>, rot=<0.2433 -0.4237 0.8278 0.2759>}
 AD_headFr_D_bt1 {parent=CR_headFr_D_bt, loc=<0 0 0>, rot=<0 0 0 1>}
 CR_headFr_E_bt {parent=CR_head_spike_fr_ct2, loc=<-2.6987 6.024 4.4188>, rot=<-0.1784 -0.2057 0.9187 0.2862>}
 AD_headFr_E_bt1 {parent=CR_headFr_E_bt, loc=<0 0 0>, rot=<0 0 0 1>}
 CR_headFr_F_bt {parent=CR_head_spike_fr_ct2, loc=<3.6219 3.6981 -2.1664>, rot=<-0.1845 0.1455 0.9621 0.1384>}
 AD_headFr_F_bt1 {parent=CR_headFr_F_bt, loc=<0 0 0>, rot=<0 0 0 1>}
 CR_faceBt_E {parent=CR_head_spike_fr_ct2, loc=<1.4614 6.2327 -0.7442>, rot=<-0.11 -0.0225 0.9751 0.1911>}
 AD_faceBt_E1 {parent=CR_faceBt_E, loc=<0 0 0>, rot=<0 0 0 1>}
 CR_head_spikes_ct2 {parent=spine_05, loc=<12.5392 -5.6488 0>, rot=<0 0 -0.036 -0.9994>}
 CR_headBa_B_bt {parent=CR_head_spikes_ct2, loc=<1.6607 -4.8156 -1.885>, rot=<0.2393 -0.8771 -0.1726 -0.3791>}
 AD_headBa_B_bt1 {parent=CR_headBa_B_bt, loc=<0 0 0>, rot=<0 0 0 1>}
 CR_headBa_C_bt {parent=CR_head_spikes_ct2, loc=<-2.4122 -5.4121 0.3303>, rot=<0.1192 -0.9361 0.0391 -0.3287>}
 AD_headBa_C_bt1 {parent=CR_headBa_C_bt, loc=<0 0 0>, rot=<0 0 0 1>}
 CR_headBa_D_bt {parent=CR_head_spikes_ct2, loc=<-3.0335 -6.437 3.3511>, rot=<0.1759 -0.9731 0.1019 -0.1084>}
 AD_headBa_D_bt1 {parent=CR_headBa_D_bt, loc=<0 0 0>, rot=<0 0 0 1>}
 CR_headBa_E_bt {parent=CR_head_spikes_ct2, loc=<-7.9323 -5.9113 5.8524>, rot=<0.3753 -0.7888 0.3586 0.3294>}
 AD_headBa_E_bt1 {parent=CR_headBa_E_bt, loc=<0 0 0>, rot=<0 0 0 1>}
 CR_headBa_F_bt {parent=CR_head_spikes_ct2, loc=<-3.8485 -4.2451 6.8999>, rot=<0.3275 -0.8678 0.2231 0.2998>}
 AD_headBa_F_bt1 {parent=CR_headBa_F_bt, loc=<0 0 0>, rot=<0 0 0 1>}
 CR_head_spikes_l3 {parent=spine_05, loc=<7.3105 -0.7241 -11.2063>, rot=<0.9994 0.036 0 0>}
 CR_shoulder_l_B {parent=CR_head_spikes_l3, loc=<5.3799 4.2908 0.6939>, rot=<0.4332 -0.4093 0.7472 0.2942>}
 AD_Lg_shoulder_B_l1 {parent=CR_shoulder_l_B, loc=<0 0 0>, rot=<0 0 0 1>}
 CR_shoulder_l_C {parent=CR_head_spikes_l3, loc=<5.6069 -1.575 2.9613>, rot=<0.337 -0.6539 0.544 0.4036>}
 AD_Lg_shoulder_C_l1 {parent=CR_shoulder_l_C, loc=<0 0 0>, rot=<0 0 0 1>}
 CR_shoulder_l_D {parent=CR_head_spikes_l3, loc=<4.6604 -7.1671 0.8322>, rot=<0.1967 -0.8088 0.3706 0.4121>}
 AD_Lg_shoulder_D_l1 {parent=CR_shoulder_l_D, loc=<0 0 0>, rot=<0 0 0 1>}
 CR_shoulder_l_E {parent=CR_head_spikes_l3, loc=<3.7562 -9.5793 -1.8126>, rot=<0.1641 -0.8948 0.3021 0.2848>}
 AD_Lg_shoulder_E_l1 {parent=CR_shoulder_l_E, loc=<0 0 0>, rot=<0 0 0 1>}
 CR_shoulder_l_A {parent=CR_head_spikes_l3, loc=<-0.2646 6.6941 3.6751>, rot=<0.4535 -0.319 0.8211 0.1353>}
 AD_Lg_shoulder_A_l1 {parent=CR_shoulder_l_A, loc=<0 0 0>, rot=<0 0 0 1>}
]
