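# Physics asset for the WeepingWoodsFestive male medium character's face accessory, setting up a dangling hat for physics simulation.
# Three identical capsules on the dyn_hat bones chain from a tiny MSTR_hat anchor, with asymmetric swing limits and zero twist range at each joint.
# A large spine_05 capsule stands in for the upper torso, linked to MSTR_hat by a constraint with free swing and twist.
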PhysicsAsset /BRCosmetics/Characters/Player/Male/Medium/Bodies/M_MED_WeepingWoodsFestive/Meshes/Parts/M_MED_WeepingWoodsFestive_FaceAcc_Physics
bodies {
 MSTR_hat [{shapeType=Capsule, length=0.505, radius=0.505, shapeName=MSTR_hat_capsule}]
 dyn_hat_01 [{shapeType=Capsule, length=9.5, radius=3, shapeName=dyn_hat_01_capsule, center=<10 0 0>, rotation=<rot -90 0 0>}]
 dyn_hat_02 [{shapeType=Capsule, length=9.5, radius=3, shapeName=dyn_hat_02_capsule, center=<10 0 0>, rotation=<rot -90 0 0>}]
 dyn_hat_03 [{shapeType=Capsule, length=9.5, radius=3, shapeName=dyn_hat_03_capsule, center=<10 0 0>, rotation=<rot -90 0 0>}]
 spine_05 [{shapeType=Capsule, length=23, radius=22, shapeName=spine_05_capsule, center=<0 1.335 0>, rotation=<rot -90 0 0>}]
}
constraints [
 {bone1Name=dyn_hat_01, bone2Name=MSTR_hat, swing1LimitDegrees=45, swing2LimitDegrees=15, twistLimitDegrees=0}
 {bone1Name=dyn_hat_02, bone2Name=dyn_hat_01, swing1LimitDegrees=45, swing2LimitDegrees=15, twistLimitDegrees=0}
 {bone1Name=dyn_hat_03, bone2Name=dyn_hat_02, swing1LimitDegrees=45, swing2LimitDegrees=15, twistLimitDegrees=0}
 {
  bone1Name=MSTR_hat
  bone2Name=spine_05
  swing1LimitDegrees=45
  swing1Motion=Free
  swing2LimitDegrees=45
  swing2Motion=Free
  twistLimitDegrees=45
  twistMotion=Free
 }
]
